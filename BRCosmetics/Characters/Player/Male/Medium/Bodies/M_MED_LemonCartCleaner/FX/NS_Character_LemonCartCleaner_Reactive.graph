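# Niagara system for the LemonCartCleaner character cosmetic that adds music-reactive sprite glows to the shirt and sleeves.
# Two emitters burst-spawn persistent sprites and use partition indices to pin each to a bone or socket on skeletal mesh interfaces.
# Music amplitude from the audio-reactive parameter collection drives sprite size and alpha, while a distance remap fades alpha out.
# Sleeve glows use two spawn groups to apply opposite X offsets.
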NiagaraSystem /BRCosmetics/Characters/Player/Male/Medium/Bodies/M_MED_LemonCartCleaner/FX/NS_Character_LemonCartCleaner_Reactive
dependencies {
 usedDataInterfaces [&NiagaraDataInterfaceCurve, &NiagaraDataInterfaceSkeletalMesh]
 usedDynamicInputs []
 usedModules [
  &self:SystemSpawnScript.SystemScriptSource.SystemScriptGraph.NiagaraNodeAssignment_1.SetVariables_#1
  &self:SystemSpawnScript.SystemScriptSource.SystemScriptGraph.NiagaraNodeAssignment_0.SetVariables_#2
  &self:SystemSpawnScript.SystemScriptSource.SystemScriptGraph.NiagaraNodeAssignment_2.SetVariables_#3
  &self:SystemSpawnScript.SystemScriptSource.SystemScriptGraph.NiagaraNodeAssignment_3.SetVariables_#4
  &self:SystemSpawnScript.SystemScriptSource.SystemScriptGraph.NiagaraNodeAssignment_4.SetVariables_#5
  &self:SystemSpawnScript.SystemScriptSource.SystemScriptGraph.NiagaraNodeAssignment_6.SetVariables_#6
  &self:SystemSpawnScript.SystemScriptSource.SystemScriptGraph.NiagaraNodeAssignment_5.SetVariables_#7
  &SystemState
 ]
}
emitters {
 ShirtReactiveGlows {
  emitterSpawnScript [/SetVariables(Emitter.NumBones=8) {isSetParameters=true}]
  emitterUpdateScript [
   &EmitterState()
   &SpawnBurst_Instantaneous(Spawn Count=$Emitter.NumBones, Use Spawn Probability=false, UseLoopCountLimit=true)
  ]
  particleSpawnScript [
   &InitializeParticle(
    Color=$System.GlowColor
    Lifetime=5
    Sprite Size Mode=::NewEnumerator3
    Uniform Sprite Size=$System.GlowSpriteSize
    UsePositionOffset=false
    Write Material Random=true
   )
   &PartitionParticles(Number Of Partitions=$Emitter.NumBones, Particles Per Partition=1, Partition Mode:ENiagara_PartitionMode=::NewEnumerator1)
   &CameraOffset(Camera Offset Amount=8)
   &DynamicMaterialParameters(
    Index 0 Param 1=3
    Index 0 Param 2=1
    Module.Index0Param0Write=true
    Module.Index0Param1Write=true
    Module.Index0Param2Write=true
    Module.Index0Param3Write=true
   )
  ]
  particleUpdateScript [
   &ParticleState(Kill Particles When Lifetime Has Elapsed=false)
   &ScaleSpriteSize(
    Scale Sprite Size Mode=::NewEnumerator0
    Uniform Scale Factor=&Lerp_Float(A=0, B=$System.GlowReactiveSpriteSize, Alpha=$NPC.NPC_Audio_Reactive_Parameters.Amplitude_Music)
   )
   &SkeletalMeshLocation(
    Bone / Socket Index=$Particles.Initial.PartitionParticles.Partition
    Bone Sampling Mode=::NewEnumerator4
    MaskSpawnGroup=false
    Skeletal Mesh=$System.Skeletal Mesh
   )
   &ScaleColor(
    Scale Alpha=&Lerp_Float(A=0, B=$System.GlowReactiveScaleColor, Alpha=$NPC.NPC_Audio_Reactive_Parameters.Amplitude_Music)
    ScaleA=true
    ScaleRGB=false
   )
   &ScaleColor(
    Scale Alpha=&FN_LOD_Distance_Remap(Start Distance=40, End Distance=1000, Minimum Output=1, Maximum Output=0, Clamp Output=true)
    ScaleA=true
    ScaleRGB=false
   ) {module=ScaleColor001}
  ]
  rendererClasses [&NiagaraSpriteRendererProperties]
  renderers [{rendererClass=&NiagaraSpriteRendererProperties}]
 }
 ShirtReactiveGlowsSleeves {
  emitterSpawnScript [/SetVariables(Emitter.NumBones=6) {isSetParameters=true}]
  emitterUpdateScript [
   &EmitterState()
   &SpawnBurst_Instantaneous(Spawn Count=3, Use Spawn Probability=false, UseLoopCountLimit=true)
   &SpawnBurst_Instantaneous(Spawn Count=3, Spawn Group=1, Use Spawn Probability=false, UseLoopCountLimit=true) {module=SpawnBurst_Instantaneous001}
  ]
  particleSpawnScript [
   &InitializeParticle(
    Color=$System.GlowColor
    Lifetime=5
    Sprite Size Mode=::NewEnumerator3
    Uniform Sprite Size=$System.GlowSpriteSize
    UsePositionOffset=false
    Write Material Random=true
   )
   &PartitionParticles(Number Of Partitions=$Emitter.NumBones, Particles Per Partition=1, Partition Mode:ENiagara_PartitionMode=::NewEnumerator1)
   &CameraOffset(Camera Offset Amount=8)
   &DynamicMaterialParameters(
    Index 0 Param 1=3
    Index 0 Param 2=1
    Module.Index0Param0Write=true
    Module.Index0Param1Write=true
    Module.Index0Param2Write=true
    Module.Index0Param3Write=true
   )
  ]
  particleUpdateScript [
   &ParticleState(Kill Particles When Lifetime Has Elapsed=false)
   &ScaleSpriteSize(
    Scale Sprite Size Mode=::NewEnumerator0
    Uniform Scale Factor=&Lerp_Float(A=0, B=$System.GlowReactiveSpriteSize, Alpha=$NPC.NPC_Audio_Reactive_Parameters.Amplitude_Music)
   )
   &SkeletalMeshLocation(
    Bone / Socket Index=$Particles.Initial.PartitionParticles.Partition
    Bone Sampling Mode=::NewEnumerator4
    MaskSpawnGroup=false
    Sampled Position Offset=&MaskVectorBySpawnGroup(Masked Vector=<5 0 0>, Passthrough Vector=<-5 0 0>)
    Skeletal Mesh=$System.Skeletal Mesh Sleeves
   )
   &ScaleColor(
    Scale Alpha=&Lerp_Float(A=0, B=$System.GlowReactiveScaleColor, Alpha=$NPC.NPC_Audio_Reactive_Parameters.Amplitude_Music)
    ScaleA=true
    ScaleRGB=false
   )
   &ScaleColor(
    Scale Alpha=&FN_LOD_Distance_Remap(Start Distance=40, End Distance=1000, Minimum Output=1, Maximum Output=0, Clamp Output=true)
    ScaleA=true
    ScaleRGB=false
   ) {module=ScaleColor001}
  ]
  rendererClasses [&NiagaraSpriteRendererProperties]
  renderers [{rendererClass=&NiagaraSpriteRendererProperties}]
 }
}
summary {
 systemName=NS_Character_LemonCartCleaner_Reactive
 emitters [
  {emitterName=ShirtReactiveGlows, rendererClasses [&NiagaraSpriteRendererProperties]}
  {emitterName=ShirtReactiveGlowsSleeves, rendererClasses [&NiagaraSpriteRendererProperties]}
 ]
 userVariables [
  User.bIsFrontEnd {type=NiagaraBool}
  User.bIsFrontEndPreview {type=NiagaraBool}
  User.isReactive {type=NiagaraFloat, defaultValue=1, description=""}
 ]
}
