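# Anime-shader post material function for the LemonCartCleaner character body cosmetic, adding animated sparkling diamonds to the finished look.
# Input gradients, colors and masks build toon shading with specular, rim, metallic, LUT, time-of-day and cinematic adjustments, output through Cloth-model attributes with optional PBR blending.
# Switchable diamonds use a view-space normal matcap and time-plus-Fresnel twinkle, overriding emissive color within their mask.
# The result passes through MF_CharacterParent_2023_PostFX.
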MaterialFunction /BRCosmetics/Characters/Player/Male/Medium/Bodies/M_MED_LemonCartCleaner/Materials/MFC_AnimeShader_LemonCartCleaner_Body_Post
graph {nodes [
  Composite() {subgraphName=Construct Base Inputs}
  FunctionInput() {inputName=---------, inputType=FunctionInput_Bool, sortPriority=30}
  FunctionInput() {inputName=--------, inputType=FunctionInput_Bool, sortPriority=20}
  n1 = Constant() {r=1}
  n2 = MaterialFunctionCall(Color_Morning=@n1, Color_Day=@n1, Color_Evening=@n1) {materialFunction=&Color_Time_of_Day}
  n3 = Multiply(A=TextureCoordinate(), B=FunctionInput() {inputName=Mask_DiamondsScale, inputType=FunctionInput_Scalar, sortPriority=1000}) {constB=8}
  n4 = TextureSample(UVs=@n3) {
   samplerSource=SSM_Wrap_WorldGroupSettings
   samplerType=SAMPLERTYPE_Normal
   texture=/BRCosmetics/Characters/Player/Male/Medium/Bodies/M_MED_LemonCartCleaner/Textures/T_LemonCartCleaner_Body_Sparkles_N
  }
  n5 = LinearInterpolate(
   A=Constant3Vector() {constant=<rgba 0 0 1 0>}
   B=@n4.RGB
   Alpha=ScalarParameter() {defaultValue=2, group=Sparkles, parameterName=Sparkles_NormalIntensity}
  )
  n6 = ComponentMask(None=Transform(None=Normalize(VectorInput=@n5)) {transformType=TRANSFORM_View}) {g=true, r=true}
  n7 = TextureSample!(UVs=Add(A=Multiply(A=@n6) {constB=0.55}) {constB=0.5}, Tex, Bias=Constant() {r=-1}, Apply View MipBias) {
   mipValueMode=TMVM_MipBias
   samplerSource=SSM_Wrap_WorldGroupSettings
   texture=/BRCosmetics/Characters/Player/Male/Medium/Bodies/M_MED_LemonCartCleaner/Textures/T_LemonCartCleaner_Sparkles_Matcap
  }
  n8 = TextureSample!(UVs=@n3, Tex, Apply View MipBias) {
   samplerSource=SSM_Wrap_WorldGroupSettings
   samplerType=SAMPLERTYPE_LinearColor
   texture=/BRCosmetics/Characters/Player/Male/Medium/Bodies/M_MED_LemonCartCleaner/Textures/T_LemonCartCleaner_Body_Sparkles_M
  }
  n9 = Ceil(None=FunctionInput() {inputName=Mask_Diamonds, inputType=FunctionInput_Scalar, sortPriority=1000})
  n10 = Add(A=Multiply(A=@n8.R, B=@n9), B=Multiply(A=NamedRerouteUsage()) {constB=0.2})
  n11 = OneMinus(None=@n8.R)
  n12 = Time()
  n13 = ScalarParameter() {defaultValue=0.2, group=Sparkles, parameterName=Sparkles_IdleSpeed}
  n14 = Add(A=Add(A=@n8.G, B=Multiply(A=@n12, B=@n13) {constB=0.1}), B=Fresnel() {baseReflectFraction=0, exponent=1})
  n15 = Power(
   Base=SphereMask(
    A=Frac(None=@n14)
    B=Constant() {r=0.5}
    Radius=ScalarParameter() {defaultValue=0.2, group=Sparkles, parameterName=Sparkles_IdleAmount, sliderMax=0.5}
   ) {attenuationRadius=0.15, hardnessPercent=0}
   Exp=ScalarParameter() {defaultValue=10, group=Sparkles, parameterName=Sparkles_IdleExponent}
  ) {constExponent=10}
  n16 = Multiply(
   A=Multiply(A=Multiply(A=@n7.RGB, B=@n10), B=ScalarParameter() {defaultValue=3, group=Sparkles, parameterName=Sparkles_MatcapBrightness})
   B=LinearInterpolate(B=ScalarParameter() {defaultValue=20, group=Sparkles, parameterName=Sparkles_IdleBrightness}, Alpha=Multiply(A=@n11, B=@n15)) {constA=1, constB=20}
  )
  n17 = ChannelMaskParameter(None=NamedRerouteUsage()) {group=03_Anime_PBR_Override, maskChannel=Alpha, parameterName=FXChannel_PBR_GMap, sortPriority=10, defaultValue=<rgba 0 0 0 1>}
  n18 = LinearInterpolate(B=OneMinus(None=@n17.R), Alpha=ScalarParameter() {group=03_Anime_PBR_Override, parameterName=PBR_GMap_Intensity, sliderMax=1}) {constA=1}
  n19 = MaterialFunctionCall!(In=NamedRerouteUsage(), LUT, LUT Multiplier) {materialFunction=/Game/Packages/Fortress_SharedMaterials/Anime_Master_Material/MF_ApplyLUT}
  n20 = NamedRerouteUsage()
  n21 = Constant()
  n22 = SetMaterialAttributes!(
   MaterialAttributes
   BaseColor=LinearInterpolate(A=Multiply(A=@n18, B=@n19.Result), B=NamedRerouteUsage(), Alpha=@n20)
   Specular=Multiply(A=NamedRerouteUsage(), B=OneMinus(None=@n20))
   Roughness=LinearInterpolate(A=NamedRerouteUsage(), Alpha=@n20)
   Metallic=NamedRerouteUsage()
   AmbientOcclusion=Constant() {r=1}
   EmissiveColor=Multiply(A=NamedRerouteUsage(), B=ScalarParameter() {group=03_Anime_PBR_Override, parameterName=PBR_Emissive_Intensity, sliderMax=1})
   Opacity=@n21
   SubsurfaceColor=@n21
   CustomData0=@n21
   Normal=NamedRerouteUsage()
  ) {attributeSetTypes [
    #1
    #2
    #3
    #4
    #5
    #6
    #7
    #8
    #9
    #10
   ]}
  n23 = NamedRerouteDeclaration(None=Saturate(None=ScalarParameter() {group=03_Anime_PBR_Override, parameterName=PBR_Shading, sliderMax=1, sortPriority=5})) {name=PBR_Switch, nodeColor=<rgba 0 1 0.8566 1>}
  n24 = Constant()
  n25 = NamedRerouteUsage()
  n26 = SetMaterialAttributes!(
   MaterialAttributes
   BaseColor=@n24
   Metallic=@n25
   Specular=@n24
   Roughness=Constant() {r=1}
   EmissiveColor=NamedRerouteUsage()
   SubsurfaceColor=MaterialFunctionCall(X=NamedRerouteUsage(), Y=@n25, Z=NamedRerouteUsage()) {materialFunction=&MakeFloat3}.Result
   CustomData0=@n24
   Normal=NamedRerouteUsage()
   AmbientOcclusion=OneMinus(None=NamedRerouteUsage())
   ShadingModel=ShadingModel() {shadingModel=MSM_Cloth}
  ) {attributeSetTypes [
    #1
    #4
    #2
    #3
    #6
    #8
    #9
    #10
    #5
    #11
   ]}
  n27 = GetMaterialAttributes(None=StaticSwitchParameter(True=BlendMaterialAttributes(A=@n26, B=@n22, Alpha=@n23), False=@n26) {group=02_Anime Parameters, parameterName=UsePBRSwitch, sortPriority=0}) {attributeGetTypes [#6]}
  n28 = NamedRerouteUsage()
  n29 = LinearInterpolate(
   A=VectorParameter() {group=Sparkles, parameterName=Sparkles_EmissiveColorRim, defaultValue=<rgba 1 1 1 1>}.RGB
   B=VectorParameter() {group=Sparkles, parameterName=Sparkles_EmissiveColorBG, defaultValue=<rgba 0 0 0 1>}.RGB
   Alpha=Power(Base=@n28, Exp=ScalarParameter() {defaultValue=0.2, group=Sparkles, parameterName=Sparkles_EmissiveColorCurve})
  )
  n30 = Multiply(A=@n28, B=ScalarParameter() {defaultValue=0.8, group=Sparkles, parameterName=Sparkles_BGOverrideAmt})
  n31 = SetMaterialAttributes!(
   MaterialAttributes=@n27.MaterialAttributes
   EmissiveColor=Add(A=Multiply(A=@n2.Result, B=@n16), B=LinearInterpolate(A=@n27.EmissiveColor, B=@n29, Alpha=@n30))
  ) {attributeSetTypes [#6]}
  n32 = Constant()
  MaterialExpressionReroute_4 = Reroute(None=@n32)
  n33 = PreSkinnedPosition()
  n34 = ShaderStageSwitch(PixelShader=VertexInterpolator(VS=@n33).PS, VertexShader=@n33)
  MaterialExpressionReroute_5 = Reroute(None=@n34)
  n35 = MaterialFunctionCall!(Material) {materialFunction=/Game/Packages/Fortress_SharedMaterials/Base_Material_Functions/CharacterBrightnessandScale_Balance}
  MaterialExpressionReroute_6 = Reroute(None=@n35.Distance Blend)
  n36 = Fresnel() {baseReflectFraction=0, exponent=1}
  MaterialExpressionReroute_7 = Reroute(None=@n36)
  FunctionOutput(None=MaterialFunctionCall(
    Material=StaticSwitchParameter(True=@n31, False=@n27.MaterialAttributes) {defaultValue=true, group=LemonCartCleaner, parameterName=Enable Diamonds}
    M=@MaterialExpressionReroute_4.M
    Pre Skinned Local Position=@MaterialExpressionReroute_5.Pre Skinned Local Position
    Distance Blend=@MaterialExpressionReroute_6.Distance Blend
    Fresnel=@MaterialExpressionReroute_7.Fresnel
   ) {materialFunction=&MF_CharacterParent_2023_PostFX}.Result)
  n37 = Ceil(None=DotProduct(A=Custom() {code="return View.DirectionalLightDirection;", description=DirectionalLightVector, inputs []}, B=PixelNormalWS()))
  NamedRerouteDeclaration(None=FunctionInput(Preview=Saturate(None=@n37)) {inputName=Shading_Gradient, inputType=FunctionInput_Scalar, sortPriority=1}) {name=Shading_Curve + Threshold, nodeColor=<rgba 1 0.1875 0 1>}
  n38 = OneMinus(None=Fresnel() {baseReflectFraction=0})
  n39 = FunctionInput(Preview=Saturate(None=Step(Y=@n38) {constX=0.9})) {inputName=RimLight_Gradient, inputType=FunctionInput_Scalar, sortPriority=11, previewValue=<1 1 1 1>}
  NamedRerouteDeclaration(None=@n39) {name=RimLight_Gradient, nodeColor=<rgba 0 1 0.3837 1>}
  n40 = FunctionInput() {inputName=Color_Shaded, sortPriority=32, previewValue=<0.18 0.18 0.18 1>}
  n41 = FunctionInput() {inputName=Color_Lit, sortPriority=31, previewValue=<0.9 0.9 0.9 1>}
  NamedRerouteDeclaration(None=LinearInterpolate(
    A=NamedRerouteDeclaration(None=@n40) {name=BC_Shaded, nodeColor=<rgba 0 0.326 1 1>}
    B=NamedRerouteDeclaration(None=@n41) {name=BC_Lit, nodeColor=<rgba 0 0.326 1 1>}
    Alpha=NamedRerouteUsage()
   )) {name=ColorOut, nodeColor=<rgba 1 0.8672 0 1>}
  n42 = Add(
   A=MaterialFunctionCall(In=NamedRerouteUsage()) {materialFunction=&MF_ApplyLUT}.Result
   B=NamedRerouteUsage()
  )
  NamedRerouteDeclaration(None=MaterialFunctionCall!(
    Emissive In=MaterialFunctionCall!(Color=@n42) {materialFunction=/Game/Packages/Fortress_SharedMaterials/Anime_Master_Material/MF_TODColorAdjustment}.Result
    LightingExtrasColorData=NamedRerouteUsage()
    CINE_LightSources=NamedRerouteUsage()
    ShadingCurve=NamedRerouteUsage()
   ) {materialFunction=/Game/Packages/Fortress_SharedMaterials/Cinematics/Anime/MF_CINE_FinalColorBlend}.Result) {name=ColorBuffer, nodeColor=<rgba 1 0.3072 0 1>}
  n43 = NamedRerouteUsage()
  n44 = MaterialFunctionCall!(Float4=NamedRerouteUsage()) {materialFunction=/Game/Global/MaterialFunctions/MF_BreakComponents}
  n45 = LinearInterpolate(B=NamedRerouteUsage(), Alpha=ScalarParameter() {group=02_Anime Parameters, parameterName=SpecularHighlight_FadeInColorLit, sliderMax=1}) {constA=1}
  n46 = Multiply(A=@n44.A, B=Multiply(A=NamedRerouteUsage(), B=NamedRerouteUsage()))
  n47 = LinearInterpolate(
   A=ScalarParameter() {defaultValue=0.2, group=02_Anime Parameters, parameterName=SpecularHighlight_ShadedIntensity, sliderMax=1}
   Alpha=NamedRerouteUsage()
  )
  n48 = LinearInterpolate(A=@n43, B=Add(A=@n43, B=Multiply(A=@n44.RGB, B=@n45)), Alpha=Saturate(None=Multiply(A=@n46, B=@n47))) {desc=apply spec}
  n49 = NamedRerouteUsage()
  n50 = NamedRerouteUsage()
  n51 = NamedRerouteUsage()
  n52 = StaticSwitchParameter(
   True=Multiply(A=ChannelMaskParameter(None=@n50) {group=02_Anime Parameters, parameterName=FXChannel_RimLight, sortPriority=10}.R, B=@n51)
   False=@n51
  ) {group=02_Anime Parameters, parameterName=UseRimLightMasking, sortPriority=0}
  n53 = ScalarParameter() {
   desc="0 = Blends in the Rimlight with the Rimlight Mask\r\n1= Adds the Rimlight color on top of the surfaace color"
   group=02_Anime Parameters
   parameterName=RimLight_AdditiveBlend
   sliderMax=1
  }
  n54 = LinearInterpolate(A=LinearInterpolate(A=@n48, B=@n49, Alpha=@n52), B=Add(A=Multiply(A=@n49, B=@n52), B=@n48), Alpha=@n53)
  NamedRerouteDeclaration(None=Saturate(None=StaticSwitchParameter(True=@n54, False=@n48) {group=02_Anime Parameters, parameterName=UseRimLight, sortPriority=0})) {name=RawColor, nodeColor=<rgba 1 0 0.7969 1>}
  NamedRerouteDeclaration(None=FunctionInput() {inputName=Thickness, inputType=FunctionInput_Scalar, sortPriority=7, previewValue=<1 1 1 1>}) {name=Thickness, nodeColor=<rgba 0.4766 0 1 1>}
  n55 = FunctionInput() {desc="RGB=Color, A=Opacity", inputName=Specular_Color, inputType=FunctionInput_Vector4, sortPriority=36, previewValue=<1 1 1 1>}
  NamedRerouteDeclaration(None=@n55) {name=SpecularColor, desc="RGB=Color, A=Opacity", nodeColor=<rgba 0 0.1989 1 1>}
  NamedRerouteDeclaration(None=FunctionInput() {inputName=NormalBuffer, sortPriority=12, previewValue=<0.5 0.5 1 1>}) {name=NormalBuffer, nodeColor=<rgba 0.2083 0.2083 1 1>}
  NamedRerouteDeclaration(None=FunctionInput() {inputName=Sheen_Mask, inputType=FunctionInput_Scalar, sortPriority=6}) {name=Sheen_Mask, nodeColor=<rgba 0.401 0.8071 1 1>}
  NamedRerouteDeclaration(None=Multiply(A=@n8.B, B=@n9)) {name=Diamonds_FlatMask, nodeColor=<rgba 0.4062 0 1 1>}
  NamedRerouteDeclaration(None=FunctionInput() {inputName=Emissive, sortPriority=34}) {name=Emissive, nodeColor=<rgba 0 0.326 1 1>}
  NamedRerouteDeclaration(None=FunctionInput() {inputName=Specular_Gradient, inputType=FunctionInput_Scalar, sortPriority=5}) {name=Specular_Curve, nodeColor=<rgba 0 1 0.1562 1>}
  NamedRerouteDeclaration(None=FunctionInput() {inputName=DFL_Color, sortPriority=35}) {name=DFL_Color, nodeColor=<rgba 0 0.326 1 1>}
  NamedRerouteDeclaration(None=LinearInterpolate(A=NamedRerouteUsage(), B=NamedRerouteUsage(), Alpha=NamedRerouteUsage())) {name=RawColor+DF, nodeColor=<rgba 0 0.0156 1 1>}
  NamedRerouteDeclaration(None=FunctionInput() {inputName=DFL_Mask, inputType=FunctionInput_Scalar, sortPriority=8}) {name=DF_Mask, nodeColor=<rgba 1 0.6562 0 1>}
  n56 = FunctionInput() {inputName=Threshold_Gradient, inputType=FunctionInput_Scalar, sortPriority=2, previewValue=<0.5 0 0 1>}
  NamedRerouteDeclaration(None=@n56) {name=Threshold_Curve, nodeColor=<rgba 0.7973 1 0 1>}
  n57 = MaterialFunctionCall(Float3=FunctionInput() {inputName=SRM, sortPriority=33}) {materialFunction=&BreakOutFloat3Components}
  NamedRerouteDeclaration(None=@n57.R) {name=MapInput_PBR_Specular, nodeColor=<rgba 1 0.0765 0 1>}
  NamedRerouteDeclaration(None=FunctionInput() {inputName=CINE_LightSources, sortPriority=21}) {name=CINE_LightSources, nodeColor=<rgba 1 0.9609 0 1>}
  NamedRerouteDeclaration(None=@n57.G) {name=MapInput_PBR_Roughness, nodeColor=<rgba 0 1 0.0508 1>}
  NamedRerouteDeclaration(None=@n57.B) {name=MapInput_PBR_Metallic, nodeColor=<rgba 0.0365 0 1 1>}
  NamedRerouteDeclaration(None=FunctionInput() {inputName=CINE_LightColor, sortPriority=22}) {name=CINE_LightColor, nodeColor=<rgba 1 0.6803 0 1>}
  NamedRerouteDeclaration(None=FunctionInput() {inputName=FX, inputType=FunctionInput_Vector4, sortPriority=9}) {name=FX, nodeColor=<rgba 1 0 0.4219 1>}
  n58 = NamedRerouteUsage()
  n59 = Multiply(A=@n58, B=NamedRerouteUsage())
  n60 = ChannelMaskParameter(None=NamedRerouteUsage()) {group=02_Anime Parameters, parameterName=FXChannel_Metallic, sortPriority=10}
  NamedRerouteDeclaration(None=StaticSwitchParameter(True=LinearInterpolate(A=@n58, B=@n59, Alpha=@n60.R) {desc=apply metallic}, False=NamedRerouteUsage()) {group=02_Anime Parameters, parameterName=UseMetallicShading, sortPriority=0}) {name=MixedColor, nodeColor=<rgba 1 0.375 0 1>}
  NamedRerouteDeclaration(None=FunctionInput() {inputName=Metallic_Gradient, inputType=FunctionInput_Scalar, sortPriority=10}) {name=Metallic_Gradient, nodeColor=<rgba 0.1107 1 0.7992 1>}
  NamedRerouteDeclaration(None=FunctionInput() {desc="RGB=Color", inputName=RimLight_Color, sortPriority=37, previewValue=<1 1 1 1>}) {name=RimLight_Color, desc="RGB=Color, A=Emissive Multiplier", nodeColor=<rgba 1 0.2256 0 1>}
  NamedRerouteUsage()
  PinBase()
  PinBase!(M=@n32, Pre Skinned Local Position=@n34, Distance Blend=@n35.Distance Blend, Fresnel=@n36) {reroutePins [
    M {expression=&self:MaterialExpressionReroute_4}
    Pre Skinned Local Position {expression=&self:MaterialExpressionReroute_5}
    Distance Blend {expression=&self:MaterialExpressionReroute_6}
    Fresnel {expression=&self:MaterialExpressionReroute_7}
   ]}
  QualitySwitch(Default=@n31, Low=@n27.MaterialAttributes) {desc="TODO: Test on lowend mobile to see if texture res holds up; otherwise disable"}
  VertexNormalWS()
 ]}
