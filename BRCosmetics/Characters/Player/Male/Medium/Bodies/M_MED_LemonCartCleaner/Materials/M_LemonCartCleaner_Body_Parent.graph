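# Parent cloth-shaded material for the LemonCartCleaner player body cosmetic, with an anime look and audio-reactive effects.
# Anime Pre and Calculations functions feed a skin-specific Post function, with switchable jersey logo and Dogg head decals tinting lit and shaded colors.
# The head decal nods by a BPM-timed rotation.
# Audio amplitude and spectrum are exposed through named reroutes for emissive boosts, including switchable sparkling diamonds that pick bands from texture variation.
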Material /BRCosmetics/Characters/Player/Male/Medium/Bodies/M_MED_LemonCartCleaner/Materials/M_LemonCartCleaner_Body_Parent
graph {
 properties {
  bAutomaticallySetUsageInEditor=false
  bHasPixelAnimation=true
  bUseMaterialAttributes=true
  bUsedWithMeshDeformer=true
  bUsedWithMorphTargets=true
  bUsedWithSkeletalMesh=true
  bUsedWithStaticLighting=true
  pixelDepthOffsetMode=PDOM_Legacy
  shadingModel=MSM_Cloth
  usedShadingModels=Cloth
 }
 nodes [
  n1 = MaterialFunctionCall(Speed=Constant() {r=0.25}) {materialFunction=&DebugTimeSine}
  n2 = LinearInterpolate(Alpha=@n1.Result) {constB=0.5}
  Arccosine!(None=@n2)
  ComponentMask(None=NamedRerouteUsage()) {a=true}
  ConstantBiasScale() {bias=-0.5, scale=2}
  Constant() {r=1}
  n3 = LinearInterpolate(Alpha=@n1.Result) {constA=0.25}
  n4 = Multiply(A=@n2) {constB=0.5}
  n5 = AppendVector(A=AppendVector(A=Arcsine!(None=@n4), B=Constant() {r=0.5}), B=Constant() {r=1})
  n6 = MaterialFunctionCall(Float2=TextureCoordinate()) {materialFunction=&BreakOutFloat2Components}
  MaterialFunctionCall(
   X=MaterialFunctionCall!(
    UVs=MaterialFunctionCall!(Texture Scale=@n3, UVs, Offset XY Texturesize Z=@n5) {materialFunction=/Engine/Functions/Engine_MaterialFunctions01/Texturing/ScaleUVsAroundPoint}.UVs
    Tiling=Constant() {r=4}
   ) {materialFunction=/Engine/ArtTools/RenderToTexture/MaterialFunctions/CheckerPattern}.Result
   Y=SphereMask(A=@n2, B=@n6.R) {attenuationRadius=0.25, hardnessPercent=0}
  ) {materialFunction=&MakeFloat3}
  n7 = Multiply(A=TextureCoordinate()) {constB=18}
  n8 = MaterialFunctionCall(UVs=Frac(None=@n7), Texture Scale=Constant() {r=1}) {materialFunction=&ScaleUVsByCenter}
  n9 = VectorNoise(World Position=AppendVector(A=@n7, B=Constant() {r=3}))
  n10 = Subtract(A=ComponentMask(None=@n9) {b=true, g=true}) {constB=0.5}
  n11 = Multiply(A=Normalize(VectorInput=@n10), B=ScalarParameter() {defaultValue=0.8, parameterName=Symbol_OffsetAmount})
  n12 = Subtract(A=@n8.UVs, B=@n11)
  n13 = MaterialFunctionCall(Float2=@n11) {materialFunction=&BreakOutFloat2Components}
  n14 = Max(A=@n13.R, B=@n13.R)
  MaterialFunctionCall(UVs=@n12, Texture Scale=OneMinus(None=Multiply(A=@n14) {constB=0.25})) {materialFunction=&ScaleUVsByCenter}
  MaterialFunctionCall(X=Add(A=@n6.R, B=@n3), Y=@n6.G) {materialFunction=&MakeFloat2}
  Multiply(A=@n10) {constB=2}
  n15 = Constant3Vector() {constant=<rgba 0 -1 0 0>}
  n16 = Transform(None=VertexNormalWS()) {transformSourceType=TRANSFORMSOURCE_World, transformType=TRANSFORM_Local}
  n17 = Subtract(A=@n12) {constB=0.5}
  n18 = ComponentMask(None=@n9) {r=true}
  n19 = Add(A=@n18, B=Multiply(A=LinearInterpolate(Alpha=@n18) {constA=0.9, constB=1.75}, B=Time()) {constA=0.2, constB=0.2})
  n20 = LinearInterpolate(Alpha=Sine(None=Multiply(A=@n19) {constB=0.2}))
  n21 = AppendVector(A=Max(A=Abs(None=@n20)) {constB=0.0001}, B=Constant() {r=1})
  n22 = Multiply(A=Sign(None=DotProduct(A=@n15, B=@n16)), B=Clamp(None=Divide(A=@n17, B=@n21)) {maxDefault=0.5, minDefault=-0.5})
  n23 = TextureSample!(
   UVs=LinearInterpolate(B=Add(A=@n22) {constB=0.5}, Alpha=@n8.0-1 mask)
   Tex
   "DDX(UVs)"=DDX(Value=@n7)
   "DDY(UVs)"=DDY(Value=@n7)
   Apply View MipBias
  ) {
   mipValueMode=TMVM_Derivative
   texture=/BRCosmetics/Characters/Player/Male/Medium/Bodies/M_MED_LemonCartCleaner/Textures/LemonCartGranite_Decal_A_2K
  }
  Multiply(A=NamedRerouteUsage(), B=@n23.R)
  n24 = MaterialFunctionCall() {materialFunction=/Game/Packages/Fortress_SharedMaterials/Base_Material_Functions/BaseCharacterShaderFunctions/Audio/MF_AudioReactivity}
  n25 = ScalarParameter() {defaultValue=1, group=LemonCartCleaner, parameterName=EnableAudioReactivity, sliderMax=1}
  n26 = Multiply(A=@n24.Amplitude, B=@n25)
  NamedRerouteDeclaration(None=Power(Base=@n26, Exp=ScalarParameter() {defaultValue=3, parameterName=Reactive_AmplitudeExp})) {name=Amplitude, nodeColor=<rgba 0.6172 1 0 1>}
  NamedRerouteDeclaration(None=Multiply(A=@n24.Spectrum, B=@n25)) {name=Spectrum, nodeColor=<rgba 0.1016 0 1 1>}
  NamedRerouteDeclaration(None=If(A=@n26, B=ScalarParameter() {defaultValue=0.05, parameterName=Reactive_AmplitudeThreshold}, "A > B"=Constant() {r=1}, "A < B"=Constant())) {name=EmoteAudioActive, nodeColor=<rgba 0 1 0.5781 1>}
  n27 = TextureSampleParameter2D() {
   parameterName=TechArtFX_Mask
   samplerType=SAMPLERTYPE_LinearColor
   texture=/BRCosmetics/Characters/Player/Male/Medium/Bodies/M_MED_LemonCartCleaner/Textures/T_LemonCartCleaner_Body_TechArtFX_Mask
  }
  NamedRerouteDeclaration(None=@n27.R) {name=Mask_DiamondsGradient, nodeColor=<rgba 0 0.3906 1 1>}
  NamedRerouteDeclaration(None=Constant()) {name=Mask_JerseyPattern, nodeColor=<rgba 0 0.3906 1 1>}
  NamedRerouteDeclaration(None=@n27.G) {name=Mask_Stripes, nodeColor=<rgba 0 0.3906 1 1>}
  NamedRerouteDeclaration(None=@n27.B) {name=Mask_Diamonds, nodeColor=<rgba 0 0.3906 1 1>}
  OneMinus(None=Multiply(A=@n1.Result) {constB=0.5})
  Sine(None=@n4)
  n28 = MaterialFunctionCall!(CustomUV) {materialFunction=/Game/Global/Materials/Character_Parent_2023/Anime_Parent_2023/MaterialFunctions/MFC_AnimeShader_2023_Pre}
  n29 = MaterialFunctionCall!(CustomLightVector, OverrideNormal_WS, STT=@n28.STT, Normal_Map=@n28.STT, DFL=@n28.STT, FX=@n28.STT) {materialFunction=/Game/Global/Materials/Character_Parent_2023/Anime_Parent_2023/MaterialFunctions/MFC_AnimeShader_2023_Calculations}
  n30 = VectorParameter() {parameterName=DoggHead_ColorLit, defaultValue=<rgba 0.7991 0.4508 0.0578 1>}
  n31 = ScalarParameter() {defaultValue=0.3488, parameterName=Logo_OffsetU, sliderMax=1, sliderMin=-1}
  n32 = ScalarParameter() {defaultValue=-0.3408, parameterName=Logo_OffsetV, sliderMax=1, sliderMin=-1}
  n33 = MaterialFunctionCall(
   UVs=Subtract(A=TextureCoordinate(), B=AppendVector(A=@n31, B=@n32))
   Texture Scale=AppendVector(A=ScalarParameter() {defaultValue=-0.11, parameterName=Logo_ScaleU}, B=ScalarParameter() {defaultValue=-0.11, parameterName=Logo_ScaleV})
  ) {materialFunction=&ScaleUVsByCenter}
  n34 = TextureSample!(UVs=@n33.UVs, Tex, Apply View MipBias) {
   desc=Logo Mask
   samplerSource=SSM_Wrap_WorldGroupSettings
   samplerType=SAMPLERTYPE_LinearColor
   texture=/BRCosmetics/Characters/Player/Male/Medium/Bodies/M_MED_LemonCartCleaner/Textures/T_LemonCartCleaner_Body_Sticker_2_FX
  }
  n35 = MaterialFunctionCall(Float3=StaticSwitchParameter(True=Multiply(A=@n33.0-1 mask, B=@n34.RGB), False=Constant()) {defaultValue=true, group=LemonCartCleaner, parameterName=Enable Jersey Logo}) {materialFunction=&BreakOutFloat3Components}
  n36 = ScalarParameter() {defaultValue=0.3488, parameterName=DoggHead_OffsetU, sliderMax=1, sliderMin=-1}
  n37 = ScalarParameter() {defaultValue=-0.2032, parameterName=DoggHead_OffsetV, sliderMax=1, sliderMin=-1}
  n38 = MaterialFunctionCall(
   UVs=Subtract(A=TextureCoordinate(), B=AppendVector(A=@n36, B=@n37))
   Texture Scale=ScalarParameter() {defaultValue=-0.15, parameterName=DoggHead_Scale}
  ) {materialFunction=&ScaleUVsByCenter}
  n39 = NamedRerouteUsage()
  n40 = Time()
  n41 = Divide(A=ScalarParameter() {defaultValue=90, parameterName=DoggHead_NodBPM}) {constB=60}
  n42 = If(
   A=Frac(None=Multiply(A=@n40, B=@n41))
   B=ScalarParameter() {defaultValue=0.5, parameterName=DoggHead_NodThreshold}
   "A > B"=ScalarParameter() {defaultValue=0.5, parameterName=DoggHead_NodAmount}
   "A < B"=Constant()
  ) {constB=0.2}
  n43 = TextureSample(UVs=Rotator(Coordinate=@n38.UVs, Time=Multiply(A=@n39, B=@n42)) {centerX=0.75, centerY=0.75}) {
   desc=Dogg Head Mask
   samplerSource=SSM_Wrap_WorldGroupSettings
   samplerType=SAMPLERTYPE_LinearColor
   texture=/BRCosmetics/Characters/Player/Male/Medium/Bodies/M_MED_LemonCartCleaner/Textures/T_LemonCartCleaner_Body_Sticker_FX
  }
  n44 = NamedRerouteUsage()
  n45 = Multiply(A=@n43.R, B=OneMinus(None=LinearInterpolate(A=@n43.G, B=@n43.G, Alpha=@n44)))
  n46 = Max(
   A=Saturate(None=Subtract(A=@n35.B, B=@n35.B))
   B=StaticSwitchParameter(True=Multiply(A=@n45, B=@n38.0-1 mask), False=Constant()) {defaultValue=true, group=LemonCartCleaner, parameterName=Enable Dogg Head}
  )
  n47 = VectorParameter() {parameterName=DoggHead_ColorShaded, defaultValue=<rgba 0.4179 0.1529 0.0006 1>}
  n48 = Multiply(
   A=NamedRerouteUsage()
   B=Multiply(
    A=VectorParameter() {parameterName=Reactive_BoostColor, defaultValue=<rgba 0.8 1 0 1>}.RGB
    B=ScalarParameter() {defaultValue=50, parameterName=Reactive_BoostBrightness}
   )
  )
  n49 = Max(A=NamedRerouteUsage(), B=@n46)
  n50 = MaterialFunctionCall!(
   Shading_Gradient=@n29.Shading_Gradient
   Threshold_Gradient=@n29.Shading_Gradient
   Specular_Gradient=@n29.Shading_Gradient
   Sheen_Mask=@n29.Shading_Gradient
   Thickness=@n29.Shading_Gradient
   DFL_Mask=@n29.Shading_Gradient
   FX=@n29.Shading_Gradient
   Metallic_Gradient=@n29.Shading_Gradient
   RimLight_Gradient=@n29.Shading_Gradient
   NormalBuffer=@n29.Shading_Gradient
   --------
   CINE_LightSources=@n29.Shading_Gradient
   CINE_LightColor=@n29.Shading_Gradient
   ---------
   Color_Lit=LinearInterpolate(
    A=LinearInterpolate(A=@n28.Color_Lit, B=@n30.RGB, Alpha=@n46)
    B=VectorParameter() {parameterName=LogoWhite_ColorLit, defaultValue=<rgba 1 1 1 1>}.RGB
    Alpha=@n35.G
   )
   Color_Shaded=LinearInterpolate(
    A=LinearInterpolate(A=@n28.Color_Shaded, B=@n47.RGB, Alpha=@n46)
    B=VectorParameter() {parameterName=LogoWhite_ColorShaded, defaultValue=<rgba 0.7 0.7 0.7 1>}.RGB
    Alpha=@n35.G
   )
   SRM=@n28.SRM
   Emissive=Add(A=@n28.Emissive, B=Multiply(A=@n48, B=@n49))
   DFL_Color=@n28.SRM
   Specular_Color=@n28.SRM
   RimLight_Color=@n28.SRM
   Mask_DiamondsScale=NamedRerouteUsage()
   Mask_Diamonds=NamedRerouteUsage()
  ) {materialFunction=&/MFC_AnimeShader_LemonCartCleaner_Body_Post}
  n51 = GetMaterialAttributes(None=@n50.Result) {attributeGetTypes [#1]}
  n52 = MaterialFunctionCall(Float4=Power(Base=NamedRerouteUsage())) {materialFunction=&BreakOutFloat4Components}
  n53 = TextureCoordinate()
  n54 = AppendVector(
   A=ScalarParameter() {group=LemonCartCleaner, parameterName=Diamonds UV Offset U}
   B=ScalarParameter() {group=LemonCartCleaner, parameterName=Diamonds UV Offset V}
  )
  n55 = Sign(None=Subtract(A=NamedRerouteUsage()) {constB=0.75})
  n56 = StaticSwitchParameter(
   True=LinearInterpolate(
    A=ScalarParameter() {defaultValue=12, group=LemonCartCleaner, parameterName=Diamonds UV Scale Secondary}
    B=ScalarParameter() {defaultValue=8, group=LemonCartCleaner, parameterName=Diamonds UV Scale}
    Alpha=Saturate(None=@n55)
   ) {constA=12, constB=8, desc=Different UV scale for Shoulders vs. Chest/Back}
   False=ScalarParameter() {defaultValue=8, group=LemonCartCleaner, parameterName=Diamonds UV Scale}
  ) {defaultValue=true, group=LemonCartCleaner, parameterName=Enable Diamonds Dual Scale}
  n57 = TextureSample(UVs=Multiply(A=Add(A=@n53, B=@n54), B=NamedRerouteDeclaration(None=@n56) {name=DiamondsUVScale, nodeColor=<rgba 0.2656 1 0 1>}) {constB=8}) {
   samplerSource=SSM_Wrap_WorldGroupSettings
   samplerType=SAMPLERTYPE_LinearColor
   texture=/BRCosmetics/Characters/Player/Male/Medium/Bodies/M_MED_LemonCartCleaner/Textures/T_LemonCartCleaner_Body_Sparkles_M
  }
  n58 = Subtract(A=@n57.G) {constB=0.25}
  n59 = LinearInterpolate(A=@n52.R, B=@n52.R, Alpha=Saturate(None=Sign(None=@n58)))
  n60 = Subtract(A=@n58) {constB=0.25}
  n61 = Saturate(None=Sign(None=@n60))
  n62 = Sign(None=Subtract(A=@n60) {constB=0.25})
  n63 = LinearInterpolate(Alpha=LinearInterpolate(A=LinearInterpolate(A=@n59, B=@n52.B, Alpha=@n61), B=@n52.A, Alpha=Saturate(None=@n62))) {constA=0.4}
  n64 = NamedRerouteUsage()
  n65 = Multiply(A=@n57.G, B=ScalarParameter() {defaultValue=0.4, parameterName=Variation})
  n66 = OneMinus(None=Saturate(None=Subtract(A=@n64, B=@n65)))
  n67 = NamedRerouteUsage()
  n68 = Multiply(
   A=Multiply(A=Subtract(A=@n63, B=@n66), B=Multiply(A=@n64, B=@n67))
   B=Multiply(A=NamedRerouteUsage(), B=Power(Base=@n57.R) {constExponent=8})
  )
  n69 = Multiply(
   A=VectorParameter() {parameterName=Reactive2_BoostColor, defaultValue=<rgba 0.7406 0.7059 1 1>}.RGB
   B=Multiply(A=Max(A=@n68) {constB=0}, B=ScalarParameter() {defaultValue=5, parameterName=Reactive2_BoostBrightness}) {constB=5}
  )
  StaticSwitchParameter(
   True=SetMaterialAttributes(MaterialAttributes=@n51.MaterialAttributes, Emissive Color=Add(A=@n51.Emissive Color, B=@n69)) {attributeSetTypes [#1]}
   False=@n51.MaterialAttributes
  ) {defaultValue=true, group=LemonCartCleaner, parameterName=Enable Diamonds Audio Reactivity}
  TextureCoordinate()
 ]
}
