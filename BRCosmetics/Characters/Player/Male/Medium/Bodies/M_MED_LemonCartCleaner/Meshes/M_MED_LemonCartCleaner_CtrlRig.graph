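# Control Rig for the LemonCartCleaner male medium body that corrects its clothing and deform bones from the body pose.
# A shared DeformRig function runs first, then remapped thigh, elbow, neck and shoulder angles, plus spherical pose readers on clavicles and upper arms, produce blend weights.
# Those weights drive parent constraints and transform offsets on belt, t-shirt, sleeve, hood, shoulder pad and upper-arm twist bones.
# A frontend-emote variable scales the center belt's pelvis constraint.
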ControlRigBlueprint /BRCosmetics/Characters/Player/Male/Medium/Bodies/M_MED_LemonCartCleaner/Meshes/M_MED_LemonCartCleaner_CtrlRig
graphs [
 RigVMFunctionLibrary {links [], nodes []}
 RigVMModel {nodes [
   RigUnit_BeginExecution = &RigUnit_BeginExecution()
   FN_DeformRig_FNC = RigVMFunctionReferenceNode(
    DeformRig_ Curve="(deformRig_l_upperarm_up_alpha_72_#1=true,deformRig_r_upperarm_up_alpha_73_#2=true,deformRig_l_upperarm_fr_alpha_74_#3=true,deformRig_r_upperarm_fr_alpha_75_CB51B80C4 ...(truncated)"
    DeformRig_Bool="(upperarm_1_#4=True,elbow_3_#5=False,wrist_7_#6=False,groin_8_#7=False,glute_17_#8=False,knee_9_40C408324C ...(truncated)"
    DeformRig_Debug=(upperarm_debug_18_#4=false,wrist_debug_20_#6=False,lat_debug_26_#9=False,groinGlute_debug_27_#7=False)
    DeformRig_Value="(upperarm_up_18_#10=(Rotation=(X=0.0,Y=-0.0,Z=0.0,W=1.0),Translation=(X=0,Y=7,Z=10),Scale3D=(X=1,Y=1,Z=0.6)),upperarm_fwd_fr_19_#11=(Rotation=(X=0.0,Y=-0.0,Z=0.0,W=1.0),Translation=(X=0.0,Y=10,Z=0.0 ...(truncated)"
    ExecuteContext=@RigUnit_BeginExecution.ExecutePin
   )
   n1 = &RigVMFunction_MathQuaternionToEuler(
    RotationOrder=ZXY
    Value=&RigUnit_GetRelativeTransformForItem(Child=(Type=Bone,Name="thigh_r"), Parent=(Type=Bone,Name="pelvis"), bChildInitial=False, bParentInitial=False).RelativeTransform.Rotation
   )
   > RigVMVariableNode(Value=&RigVMFunction_MathDoubleRemap(SourceMaximum=-10, SourceMinimum=0, TargetMaximum=0, TargetMinimum=1, Value=@n1.Result.Y).Result) {variable=R thighInward reverseAlpha}
   Absolute = &RigVMFunction_MathDoubleAbs(Value=@n1.Result.Z)
   Subtract = &RigVMFunction_MathDoubleSub(A=180, B=@Absolute.Result)
   > RigVMVariableNode(Value=&RigVMFunction_MathDoubleRemap(SourceMaximum=40, SourceMinimum=0, TargetMaximum=1, TargetMinimum=0, Value=@Subtract.Result).Result) {variable=R thighForward alpha}
   n2 = &RigVMFunction_MathQuaternionToEuler(
    RotationOrder=ZXY
    Value=&RigUnit_GetRelativeTransformForItem(Child=(Type=Bone,Name="thigh_l"), Parent=(Type=Bone,Name="pelvis"), bChildInitial=False, bParentInitial=False).RelativeTransform.Rotation
   )
   > RigVMVariableNode(Value=&RigVMFunction_MathDoubleRemap(SourceMaximum=10, SourceMinimum=0, TargetMaximum=0, TargetMinimum=1, Value=@n2.Result.Y).Result) {variable=L thighInward reverseAlpha}
   > RigVMVariableNode(Value=&RigVMFunction_MathDoubleRemap(SourceMaximum=-40, SourceMinimum=0, TargetMaximum=1, TargetMinimum=0, Value=@n2.Result.Z).Result) {variable=L thighForward alpha}
   n3 = &RigUnit_GetRelativeTransformForItem(Child=(Type=Bone,Name="lowerarm_l"), Parent=(Type=Bone,Name="upperarm_l"), bChildInitial=False, bParentInitial=False)
   > RigVMVariableNode(Value=&RigVMFunction_MathDoubleRemap(
     SourceMaximum=110
     SourceMinimum=50
     TargetMaximum=1
     TargetMinimum=0
     Value=&RigVMFunction_MathQuaternionToEuler(RotationOrder=XYZ, Value=@n3.RelativeTransform.Rotation).Result.Z
    ).Result) {variable=L lowerarm bend}
   n4 = &RigUnit_GetRelativeTransformForItem(Child=(Type=Bone,Name="lowerarm_r"), Parent=(Type=Bone,Name="upperarm_r"), bChildInitial=False, bParentInitial=False)
   > RigVMVariableNode(Value=&RigVMFunction_MathDoubleRemap(
     SourceMaximum=110
     SourceMinimum=50
     TargetMaximum=1
     TargetMinimum=0
     Value=&RigVMFunction_MathQuaternionToEuler(RotationOrder=XYZ, Value=@n4.RelativeTransform.Rotation).Result.Z
    ).Result) {variable=R lowerarm bend}
   n5 = &RigUnit_GetRelativeTransformForItem(Child=(Type=Bone,Name="neck_02"), Parent=(Type=Bone,Name="spine_05"), bChildInitial=False, bParentInitial=False)
   > RigVMVariableNode(Value=&RigVMFunction_MathDoubleRemap(
     SourceMaximum=-5
     SourceMinimum=10
     TargetMaximum=1
     TargetMinimum=0
     Value=&RigVMFunction_MathQuaternionToEuler(RotationOrder=ZXY, Value=@n5.RelativeTransform.Rotation).Result.Z
    ).Result) {variable=neckBack alpha}
   > &RigUnit_ParentConstraint(
    AdvancedSettings=(InterpolationType=Average,RotationOrderForFilter=XZY)
    Child=(Type=Bone,Name="dyn_boneBelt_front_1")
    Filter=(TranslationFilter=(bX=true,bY=true,bZ=true),RotationFilter=(bX=true,bY=true,bZ=true),ScaleFilter=(bX=True,bY=True,bZ=True))
    Parents=((Item=(Type=Bone,Name="thigh_r"),Weight=1))
    Weight=0.9
   )
   > &RigUnit_ParentConstraint(
    AdvancedSettings=(InterpolationType=Average,RotationOrderForFilter=XZY)
    Child=(Type=Bone,Name="dyn_boneBelt_center_1")
    Filter=(TranslationFilter=(bX=true,bY=true,bZ=true),RotationFilter=(bX=True,bY=True,bZ=True),ScaleFilter=(bX=True,bY=True,bZ=True))
    Parents=((Item=(Type=Bone,Name="thigh_r"),Weight=1))
    Weight=1
   )
   > &RigUnit_ParentConstraint(
    AdvancedSettings=(InterpolationType=Average,RotationOrderForFilter=XZY)
    Child=(Type=Bone,Name="dyn_boneBelt_center_1")
    Filter=(TranslationFilter=(bX=true,bY=true,bZ=true),RotationFilter=(bX=True,bY=True,bZ=True),ScaleFilter=(bX=True,bY=True,bZ=True))
    Parents=((Item=(Type=Bone,Name="pelvis"),Weight=1))
    Weight=&RigVMFunction_MathDoubleMul(B=0.6, A=RigVMVariableNode() {variable=isNotPlayingEmoteInFrontend}.Value).Result
   )
   > &RigUnit_ModifyTransforms(
    ItemToModify=((Item=(Type=Bone,Name="dyn_boneBelt_back_1"),Transform=(Rotation=(X=-0.0338,Y=-0.1261,Z=-0.2566,W=0.9577),Translation=(X=0,Y=0,Z=0),Scale3D=(X=1,Y=1,Z=1))))
    WeightMinimum=0
    Weight=RigVMVariableNode() {variable=R thighForward alpha}.Value
   )
   > &RigUnit_ModifyTransforms(
    ItemToModify=((Item=(Type=Bone,Name="dyn_L_sleeveScoot"),Transform=(Rotation=(X=0,Y=0,Z=0,W=1),Translation=(X=-0.5,Y=0,Z=0),Scale3D=(X=0.3,Y=1,Z=1))))
    WeightMinimum=0
    Weight=RigVMVariableNode() {variable=L lowerarm bend}.Value
   )
   > &RigUnit_ModifyTransforms(
    ItemToModify=((Item=(Type=Bone,Name="dyn_R_sleeveScoot"),Transform=(Rotation=(X=0,Y=0,Z=0,W=1),Translation=(X=0.5,Y=1,Z=0),Scale3D=(X=0.3,Y=1,Z=1))))
    WeightMinimum=0
    Weight=RigVMVariableNode() {variable=R lowerarm bend}.Value
   )
   > &RigUnit_ParentConstraint(
    AdvancedSettings=(InterpolationType=Average,RotationOrderForFilter=XZY)
    Child=(Type=Bone,Name="dyn_hoodWobble")
    Filter=(TranslationFilter=(bX=false,bY=false,bZ=false),RotationFilter=(bX=True,bY=True,bZ=True),ScaleFilter=(bX=True,bY=True,bZ=True))
    Parents=((Item=(Type=Bone,Name="neck_01"),Weight=1))
    Weight=RigVMVariableNode() {variable=neckBack alpha}.Value
   )
   n6 = RigVMVariableNode() {variable=L thighForward alpha}
   > &RigUnit_ParentConstraint(
    AdvancedSettings=(InterpolationType=Average,RotationOrderForFilter=XZY)
    Child=(Type=Bone,Name="dyn_tshirt_front_l_MSTR")
    Filter=(TranslationFilter=(bX=false,bY=false,bZ=false),RotationFilter=(bX=True,bY=True,bZ=True),ScaleFilter=(bX=True,bY=True,bZ=True))
    Parents=((Item=(Type=Bone,Name="thigh_l"),Weight=1))
    Weight=@n6.Value
   )
   n7 = &RigVMFunction_MathDoubleAdd(A=RigVMVariableNode() {variable=L thighInward reverseAlpha}.Value, B=RigVMVariableNode() {variable=L thighForward alpha}.Value)
   > &RigUnit_ParentConstraint(
    AdvancedSettings=(InterpolationType=Average,RotationOrderForFilter=XZY)
    Child=(Type=Bone,Name="dyn_tshirt_side_l_MSTR")
    Filter=(TranslationFilter=(bX=false,bY=false,bZ=false),RotationFilter=(bX=True,bY=true,bZ=true),ScaleFilter=(bX=True,bY=True,bZ=True))
    Parents=((Item=(Type=Bone,Name="thigh_l"),Weight=1))
    Weight=@n7.Result
   )
   > &RigUnit_ModifyTransforms(
    ItemToModify="((Item=(Type=Bone,Name=\"dyn_tshirt_side_l_MSTR\"),Transform=(Rotation=(X=0,Y=0,Z=-0.1736,W=0.9848),Translation=(X=0,Y=1,Z=0),Scale3D=(X=1,Y=1,Z=1))),(Item=(Type=Bone,Name=\"dyn_tshirt_rear_l_MSTR\"),Transform=(Rotation=(X=0,Y=0,Z=-0.0436,W=0.999),Translation= ...(truncated)"
    WeightMinimum=0
    Weight=&RigVMFunction_MathDoubleSub(A=1, B=RigVMVariableNode() {variable=L thighInward reverseAlpha}.Value).Result
   )
   n8 = RigVMVariableNode() {variable=L thighForward alpha}
   > &RigUnit_ParentConstraint(
    AdvancedSettings=(InterpolationType=Average,RotationOrderForFilter=XZY)
    Child=(Type=Bone,Name="dyn_tshirt_rear_l_MSTR")
    Filter=(TranslationFilter=(bX=false,bY=false,bZ=false),RotationFilter=(bX=True,bY=true,bZ=true),ScaleFilter=(bX=True,bY=True,bZ=True))
    Parents=((Item=(Type=Bone,Name="thigh_l"),Weight=1))
    Weight=&RigVMFunction_MathDoubleMul(A=0.2, B=@n8.Value).Result
   )
   > &RigUnit_ModifyTransforms(
    ItemToModify=((Item=(Type=Bone,Name="dyn_tshirt_rear_l_MSTR"),Transform=(Rotation=(X=0,Y=0,Z=-0.1305,W=0.9914),Translation=(X=0,Y=0,Z=0),Scale3D=(X=1,Y=1,Z=1))))
    WeightMinimum=0
    Weight=@n8.Value
   )
   n9 = RigVMVariableNode() {variable=R thighForward alpha}
   > &RigUnit_ParentConstraint(
    AdvancedSettings=(InterpolationType=Average,RotationOrderForFilter=XZY)
    Child=(Type=Bone,Name="dyn_tshirt_front_r_MSTR")
    Filter=(TranslationFilter=(bX=false,bY=false,bZ=false),RotationFilter=(bX=True,bY=True,bZ=True),ScaleFilter=(bX=True,bY=True,bZ=True))
    Parents=((Item=(Type=Bone,Name="thigh_r"),Weight=1))
    Weight=@n9.Value
   )
   Add = &RigVMFunction_MathDoubleAdd(A=RigVMVariableNode() {variable=R thighInward reverseAlpha}.Value, B=RigVMVariableNode() {variable=R thighForward alpha}.Value)
   > &RigUnit_ParentConstraint(
    AdvancedSettings=(InterpolationType=Average,RotationOrderForFilter=XZY)
    Child=(Type=Bone,Name="dyn_tshirt_side_r_MSTR")
    Filter=(TranslationFilter=(bX=false,bY=false,bZ=false),RotationFilter=(bX=true,bY=true,bZ=true),ScaleFilter=(bX=True,bY=True,bZ=True))
    Parents=((Item=(Type=Bone,Name="thigh_r"),Weight=1))
    Weight=@Add.Result
   )
   > &RigUnit_ModifyTransforms(
    ItemToModify="((Item=(Type=Bone,Name=\"dyn_tshirt_front_r_MSTR\"),Transform=(Rotation=(X=0,Y=0,Z=0.1205,W=0.9927),Translation=(X=0,Y=-0.0888,Z=-0.3406),Scale3D=(X=1,Y=1,Z=1))),(Item=(Type=Bone,Name=\"dyn_tshirt_side_r_MSTR\"),Transform=(Rotation=(X=0.1603,Y=-0.0818,Z=0.0133 ...(truncated)"
    WeightMinimum=0
    Weight=@n9.Value
   )
   > &RigUnit_ModifyTransforms(
    ItemToModify="((Item=(Type=Bone,Name=\"dyn_tshirt_side_r_MSTR\"),Transform=(Rotation=(X=0,Y=0,Z=-0.1736,W=0.9848),Translation=(X=0,Y=0,Z=0),Scale3D=(X=1,Y=1,Z=1))),(Item=(Type=Bone,Name=\"dyn_tshirt_front_r_MSTR\"),Transform=(Rotation=(X=0,Y=0,Z=-0.0436,W=0.999),Translation ...(truncated)"
    WeightMinimum=0
    Weight=&RigVMFunction_MathDoubleSub(A=1, B=RigVMVariableNode() {variable=R thighInward reverseAlpha}.Value).Result
   )
   n10 = RigVMVariableNode() {variable=R thighForward alpha}
   Multiply = &RigVMFunction_MathDoubleMul(A=0.2, B=@n10.Value)
   > &RigUnit_ParentConstraint(
    AdvancedSettings=(InterpolationType=Average,RotationOrderForFilter=XZY)
    Child=(Type=Bone,Name="dyn_tshirt_rear_r_MSTR")
    Filter=(TranslationFilter=(bX=false,bY=false,bZ=false),RotationFilter=(bX=True,bY=true,bZ=true),ScaleFilter=(bX=True,bY=True,bZ=True))
    Parents=((Item=(Type=Bone,Name="thigh_r"),Weight=1))
    Weight=@Multiply.Result
   )
   > &RigUnit_ModifyTransforms(
    ItemToModify=((Item=(Type=Bone,Name="dyn_tshirt_rear_r_MSTR"),Transform=(Rotation=(X=0,Y=0,Z=-0.1564,W=0.9877),Translation=(X=0,Y=0,Z=0),Scale3D=(X=1,Y=1,Z=1))))
    WeightMinimum=0
    Weight=@n10.Value
   )
   > &RigUnit_ParentConstraint(
    AdvancedSettings=(InterpolationType=Average,RotationOrderForFilter=XZY)
    Child=(Type=Bone,Name="dyn_tshirt_front_MSTR")
    Filter=(TranslationFilter=(bX=false,bY=false,bZ=false),RotationFilter=(bX=True,bY=True,bZ=True),ScaleFilter=(bX=True,bY=True,bZ=True))
    Parents=((Item=(Type=Bone,Name="dyn_tshirt_front_l_01"),Weight=0.5),(Item=(Type=Bone,Name="dyn_tshirt_front_r_01"),Weight=0.5))
    Weight=1
   )
   > &RigUnit_ParentConstraint(
    AdvancedSettings=(InterpolationType=Average,RotationOrderForFilter=XZY)
    Child=(Type=Bone,Name="dyn_tshirt_front_02")
    Filter=(TranslationFilter=(bX=false,bY=false,bZ=false),RotationFilter=(bX=True,bY=True,bZ=True),ScaleFilter=(bX=True,bY=True,bZ=True))
    Parents=((Item=(Type=Bone,Name="dyn_tshirt_front_l_02"),Weight=0.5),(Item=(Type=Bone,Name="dyn_tshirt_front_r_02"),Weight=0.5))
    Weight=1
   )
   > n11 = &RigUnit_SphericalPoseReader(
    ActiveRegionScaleFactors=(PositiveWidth=1,NegativeWidth=1,PositiveHeight=1,NegativeHeight=1)
    ActiveRegionSize=0.1
    Debug=(bDrawDebug=false,bDraw2D=False,bDrawLocalAxes=False,DebugScale=25,DebugSegments=20,DebugThickness=0.25)
    DriverAxis=(X=1,Y=0,Z=0)
    DriverItem=(Type=Bone,Name="clavicle_l")
    FalloffRegionScaleFactors=(PositiveWidth=1,NegativeWidth=1,PositiveHeight=1,NegativeHeight=1)
    FalloffSize=0.3
    FlipHeightScaling=False
    FlipWidthScaling=False
    OptionalParentItem=(Type=Bone,Name="None")
    RotationOffset=(X=0,Y=-50,Z=0)
   )
   > RigVMVariableNode(Value=@n11.OutputParam) {variable=spr_clavicle_l_raise}
   > n12 = &RigUnit_SphericalPoseReader(
    ActiveRegionScaleFactors=(PositiveWidth=1,NegativeWidth=1,PositiveHeight=1,NegativeHeight=1)
    ActiveRegionSize=0.1
    Debug=(bDrawDebug=false,bDraw2D=False,bDrawLocalAxes=False,DebugScale=25,DebugSegments=20,DebugThickness=0.25)
    DriverAxis=(X=-1,Y=0,Z=0)
    DriverItem=(Type=Bone,Name="clavicle_r")
    FalloffRegionScaleFactors=(PositiveWidth=1,NegativeWidth=1,PositiveHeight=1,NegativeHeight=1)
    FalloffSize=0.3
    FlipHeightScaling=False
    FlipWidthScaling=False
    OptionalParentItem=(Type=Bone,Name="None")
    RotationOffset=(X=180,Y=-50,Z=0)
   )
   > RigVMVariableNode(Value=@n12.OutputParam) {variable=spr_clavicle_r_raise}
   > n13 = &RigUnit_SphericalPoseReader(
    ActiveRegionScaleFactors=(PositiveWidth=1.5,NegativeWidth=1.5,PositiveHeight=1.5,NegativeHeight=1.5)
    ActiveRegionSize=0.12
    Debug=(bDrawDebug=false,DebugScale=25,DebugSegments=20,DebugThickness=0.25)
    DriverAxis=(X=1,Y=0,Z=0)
    DriverItem=(Type=Bone,Name="upperarm_l")
    FalloffRegionScaleFactors=(PositiveWidth=1,NegativeWidth=1,PositiveHeight=1,NegativeHeight=1)
    FalloffSize=0.6
    OptionalParentItem=(Type=Bone,Name="clavicle_l")
    RotationOffset=(X=50,Y=0,Z=0)
   )
   > RigVMVariableNode(Value=@n13.OutputParam) {variable=spr_arm_l_fwd}
   > n14 = &RigUnit_SphericalPoseReader(
    ActiveRegionScaleFactors=(PositiveWidth=1.5,NegativeWidth=1.5,PositiveHeight=1.5,NegativeHeight=1.5)
    ActiveRegionSize=0.12
    Debug=(bDrawDebug=false,DebugScale=25,DebugSegments=20,DebugThickness=0.25)
    DriverAxis=(X=-1,Y=0,Z=0)
    DriverItem=(Type=Bone,Name="upperarm_r")
    FalloffRegionScaleFactors=(PositiveWidth=1,NegativeWidth=1,PositiveHeight=1,NegativeHeight=1)
    FalloffSize=0.6
    FlipHeightScaling=true
    FlipWidthScaling=false
    OptionalParentItem=(Type=Bone,Name="clavicle_r")
    RotationOffset=(X=50,Y=180,Z=180)
   )
   > RigVMVariableNode(Value=@n14.OutputParam) {variable=spr_arm_r_fwd}
   > &RigUnit_ModifyTransforms(
    ItemToModify=((Item=(Type=Bone,Name="deform_shoulderpad_l_MSTR"),Transform=(Rotation=(X=0,Y=0,Z=0,W=1),Translation=(X=2,Y=0,Z=0),Scale3D=(X=1,Y=1,Z=1))))
    WeightMinimum=0
    Weight=RigVMVariableNode() {variable=spr_clavicle_l_raise}.Value
   )
   > &RigUnit_ModifyTransforms(
    ItemToModify="((Item=(Type=Bone,Name=\"deform_shoulderpad_l_MSTR\"),Transform=(Rotation=(X=0,Y=0,Z=0,W=1),Translation=(X=-4,Y=0,Z=3.5),Scale3D=(X=1,Y=1.1,Z=1))),(Item=(Type=Bone,Name=\"deform_shoulderpad_front_l\"),Transform=(Rotation=(X=0.2663,Y=0.3978,Z=-0.0164,W=0.8778), ...(truncated)"
    WeightMinimum=0
    Weight=RigVMVariableNode() {variable=spr_arm_l_fwd}.Value
   )
   > &RigUnit_ModifyTransforms(
    ItemToModify=((Item=(Type=Bone,Name="deform_shoulderpad_r_MSTR"),Transform=(Rotation=(X=0,Y=0,Z=0,W=1),Translation=(X=-2,Y=0,Z=0),Scale3D=(X=1,Y=1,Z=1))))
    WeightMinimum=0
    Weight=RigVMVariableNode() {variable=spr_clavicle_r_raise}.Value
   )
   > &RigUnit_ModifyTransforms(
    ItemToModify="((Item=(Type=Bone,Name=\"deform_shoulderpad_r_MSTR\"),Transform=(Rotation=(X=0,Y=0,Z=0,W=1),Translation=(X=4,Y=0,Z=-3.5),Scale3D=(X=1,Y=1.1,Z=1))),(Item=(Type=Bone,Name=\"deform_shoulderpad_front_r\"),Transform=(Rotation=(X=0.2663,Y=0.3978,Z=-0.0164,W=0.8778), ...(truncated)"
    WeightMinimum=0
    Weight=RigVMVariableNode() {variable=spr_arm_r_fwd}.Value
   )
   > &RigUnit_ModifyTransforms(
    ItemToModify=((Item=(Type=Bone,Name="upperarm_twist_01_l"),Transform=(Rotation=(X=0.2164,Y=0,Z=0,W=0.9763),Translation=(X=0,Y=0,Z=0),Scale3D=(X=1,Y=1,Z=1))))
    WeightMinimum=0
    Weight=RigVMVariableNode() {variable=spr_arm_l_fwd}.Value
   )
   > &RigUnit_ModifyTransforms(
    ItemToModify=((Item=(Type=Bone,Name="upperarm_twist_01_r"),Transform=(Rotation=(X=0.2164,Y=0,Z=0,W=0.9763),Translation=(X=0,Y=0,Z=0),Scale3D=(X=1,Y=1,Z=1))))
    WeightMinimum=0
    Weight=RigVMVariableNode() {variable=spr_arm_r_fwd}.Value
   )
   n15 = &RigVMFunction_MathDoubleRemap(
    SourceMaximum=70
    SourceMinimum=30
    TargetMaximum=1
    TargetMinimum=0
    Value=&RigVMFunction_MathQuaternionToEuler(
     RotationOrder=ZXY
     Value=&RigUnit_GetRelativeTransformForItem(Child=(Type=Bone,Name="upperarm_l"), Parent=(Type=Bone,Name="clavicle_l"), bChildInitial=False, bParentInitial=False).RelativeTransform.Rotation
    ).Result.X
   )
   > &RigUnit_ParentConstraint(
    AdvancedSettings=(InterpolationType=Average,RotationOrderForFilter=XZY)
    Child=(Type=Bone,Name="dyn_L_backSleeve_1")
    Filter=(TranslationFilter=(bX=true,bY=true,bZ=true),RotationFilter=(bX=True,bY=True,bZ=True),ScaleFilter=(bX=True,bY=True,bZ=True))
    Parents=((Item=(Type=Bone,Name="upperarm_twist_01_l"),Weight=0.5),(Item=(Type=Bone,Name="upperarm_l"),Weight=0.5))
    Weight=@n15.Result
   )
   > &RigUnit_ParentConstraint(
    AdvancedSettings=(InterpolationType=Average,RotationOrderForFilter=XZY)
    Child=(Type=Bone,Name="dyn_L_aboveSleeve_1")
    Filter=(TranslationFilter=(bX=true,bY=true,bZ=true),RotationFilter=(bX=True,bY=True,bZ=True),ScaleFilter=(bX=True,bY=True,bZ=True))
    Parents=((Item=(Type=Bone,Name="upperarm_twist_01_l"),Weight=0.5),(Item=(Type=Bone,Name="upperarm_l"),Weight=0.5))
    Weight=@n15.Result
   )
   > &RigUnit_ParentConstraint(
    AdvancedSettings=(InterpolationType=Average,RotationOrderForFilter=XZY)
    Child=(Type=Bone,Name="dyn_L_underSleeve_1")
    Filter=(TranslationFilter=(bX=true,bY=true,bZ=true),RotationFilter=(bX=True,bY=True,bZ=True),ScaleFilter=(bX=True,bY=True,bZ=True))
    Parents=((Item=(Type=Bone,Name="upperarm_twist_01_l"),Weight=0.5),(Item=(Type=Bone,Name="upperarm_l"),Weight=0.5))
    Weight=@n15.Result
   )
   n16 = &RigVMFunction_MathDoubleRemap(
    SourceMaximum=70
    SourceMinimum=30
    TargetMaximum=1
    TargetMinimum=0
    Value=&RigVMFunction_MathQuaternionToEuler(
     RotationOrder=ZXY
     Value=&RigUnit_GetRelativeTransformForItem(Child=(Type=Bone,Name="upperarm_r"), Parent=(Type=Bone,Name="clavicle_r"), bChildInitial=False, bParentInitial=False).RelativeTransform.Rotation
    ).Result.X
   )
   > &RigUnit_ParentConstraint(
    AdvancedSettings=(InterpolationType=Average,RotationOrderForFilter=XZY)
    Child=(Type=Bone,Name="dyn_R_backSleeve_1")
    Filter=(TranslationFilter=(bX=true,bY=true,bZ=true),RotationFilter=(bX=True,bY=True,bZ=True),ScaleFilter=(bX=True,bY=True,bZ=True))
    Parents=((Item=(Type=Bone,Name="upperarm_twist_01_r"),Weight=0.5),(Item=(Type=Bone,Name="upperarm_r"),Weight=0.5))
    Weight=@n16.Result
   )
   > &RigUnit_ParentConstraint(
    AdvancedSettings=(InterpolationType=Average,RotationOrderForFilter=XZY)
    Child=(Type=Bone,Name="dyn_R_aboveSleeve_1")
    Filter=(TranslationFilter=(bX=true,bY=true,bZ=true),RotationFilter=(bX=True,bY=True,bZ=True),ScaleFilter=(bX=True,bY=True,bZ=True))
    Parents=((Item=(Type=Bone,Name="upperarm_twist_01_r"),Weight=0.5),(Item=(Type=Bone,Name="upperarm_r"),Weight=0.5))
    Weight=@n16.Result
   )
   > &RigUnit_ParentConstraint(
    AdvancedSettings=(InterpolationType=Average,RotationOrderForFilter=XZY)
    Child=(Type=Bone,Name="dyn_R_underSleeve_1")
    Filter=(TranslationFilter=(bX=true,bY=true,bZ=true),RotationFilter=(bX=True,bY=True,bZ=True),ScaleFilter=(bX=True,bY=True,bZ=True))
    Parents=((Item=(Type=Bone,Name="upperarm_twist_01_r"),Weight=0.5),(Item=(Type=Bone,Name="upperarm_r"),Weight=0.5))
    Weight=@n16.Result
   )
   &RigUnit_AimConstraintLocalSpaceOffset(
    Child=(Type=Bone,Name="dyn_boneBelt_back_1")
    ExecutePin=()
    Filter=(bX=True,bY=True,bZ=True)
    Parents=((Item=(Type=Control,Name="dyn_boneBelt_back_2_ctrl"),Weight=1))
    Weight=1
    WorldUp=(Target=(X=0,Y=0,Z=1),Kind=Location,Space=(Type=Control,Name="dyn_boneBelt_back_2_ctrl"))
   )
   &RigUnit_ModifyTransforms(
    ExecutePin=()
    ItemToModify="((Item=(Type=Bone,Name=\"dyn_tshirt_front_l_MSTR\"),Transform=(Rotation=(X=0.0977,Y=0.0618,Z=0.0202,W=0.9931),Translation=(X=-1.3194,Y=1.1146,Z=-5.3914),Scale3D=(X=1,Y=1,Z=1))),(Item=(Type=Bone,Name=\"dyn_tshirt_side_l_MSTR\"),Transform=(Rotation=(X=0,Y=0,Z=0, ...(truncated)"
    WeightMinimum=0
    Weight=@n6.Value
   )
   &RigUnit_ModifyTransforms(
    ExecutePin=()
    ItemToModify="((Item=(Type=Bone,Name=\"dyn_boneBelt_back_2\"),Transform=(Rotation=(X=0,Y=0,Z=0,W=1),Translation=(X=0,Y=0,Z=4),Scale3D=(X=1,Y=1,Z=1))),(Item=(Type=Bone,Name=\"dyn_boneBelt_center_1\"),Transform=(Rotation=(X=0,Y=0.0436,Z=0,W=0.999),Translation=(X=0,Y=0,Z=2),Sc ...(truncated)"
    Weight=1
    WeightMinimum=0
   )
   &RigVMFunction_MathDoubleMul(B=0.5, A=@n7.Result)
   &RigVMFunction_MathDoubleMul(B=0.5, A=@Add.Result)
  ]}
]
hierarchy [
 root {loc=<0 0 0>, rot=<0 0 0 1>}
 pelvis {parent=root, loc=<0 -0.7071 93.1191>, rot=<0.0127 0.707 -0.0127 -0.707>}
 spine_01 {parent=pelvis, loc=<6.7494 -0.6388 0>, rot=<0 0 0.034 -0.9994>}
 spine_02 {parent=spine_01, loc=<7.2518 -0.035 0>, rot=<0 0 0.0238 -0.9997>}
 spine_03 {parent=spine_02, loc=<6.0104 0 0>, rot=<0 0 0.0149 -0.9999>}
 spine_04 {parent=spine_03, loc=<8.1024 0 0>, rot=<0 0 -0.0132 -0.9999>}
 spine_05 {parent=spine_04, loc=<10.0146 0 0>, rot=<0 0 -0.0015 -1>}
 attach_cape {parent=spine_05, loc=<9.017 -11.714 0>, rot=<0.0616 -0.7044 -0.0616 -0.7044>}
 attach_backpack {parent=spine_05, loc=<-0.123 -13.473 0>, rot=<-0.0093 -0.707 0.0093 -0.707>}
 clavicle_l {parent=spine_05, loc=<4.3575 -0.1466 -2.3223>, rot=<0.0465 -0.7289 -0.009 -0.6829>}
 upperarm_l {parent=clavicle_l, loc=<15.5111 0 0>, rot=<0.0226 -0.3234 -0.0088 -0.9459>}
 lowerarm_l {parent=upperarm_l, loc=<26.6514 0 0>, rot=<0.0391 -0.0099 -0.2453 -0.9686>}
 hand_l {parent=lowerarm_l, loc=<22.3491 0 0>, rot=<-0.7081 -0.0507 0.0975 -0.6975>}
 thumb_01_l {parent=hand_l, loc=<1.3118 -0.3191 -2.459>, rot=<0.6546 -0.4042 0.0664 -0.6354>}
 thumb_02_l {parent=thumb_01_l, loc=<5.9336 -0.0448 0.0401>, rot=<-0.0037 0.0213 0.298 -0.9543>}
 thumb_03_l {parent=thumb_02_l, loc=<2.2223 0.3682 -0.0907>, rot=<0.0118 -0.013 0.0066 -0.9998>}
 index_metacarpal_l {parent=hand_l, loc=<1.6535 0.0183 -2.96>, rot=<0.0158 -0.0124 -0.0173 -0.9996>}
 index_01_l {parent=index_metacarpal_l, loc=<8.3008 0.0357 -0.0524>, rot=<-0.0044 0.0485 0.0924 -0.9945>}
 index_02_l {parent=index_01_l, loc=<3.3587 -0.3004 0.0452>, rot=<-0.002 0.0027 0.3225 -0.9466>}
 index_03_l {parent=index_02_l, loc=<2.5578 -0.123 -0.0252>, rot=<-0.002 0.0067 0.2558 -0.9667>}
 middle_metacarpal_l {parent=hand_l, loc=<2.1856 -0.1725 -0.5984>, rot=<0.006 0.0168 -0.0329 -0.9993>}
 middle_01_l {parent=middle_metacarpal_l, loc=<7.3275 -0.2447 0.0129>, rot=<-0.0208 0.0042 0.0858 -0.9961>}
 middle_02_l {parent=middle_01_l, loc=<4.0123 0.1401 0.1175>, rot=<-0.0218 0.0258 0.3512 -0.9357>}
 middle_03_l {parent=middle_02_l, loc=<2.6077 0.0534 0.1373>, rot=<0.0082 0.0188 0.1708 -0.9851>}
 ring_metacarpal_l {parent=hand_l, loc=<2.2074 -0.3625 1.3638>, rot=<-0.0284 0.0722 -0.0458 -0.9959>}
 ring_01_l {parent=ring_metacarpal_l, loc=<7.0295 -0.6197 -0.1446>, rot=<-0.0489 -0.0252 0.1892 -0.9804>}
 ring_02_l {parent=ring_01_l, loc=<3.6174 0.4647 -0.1027>, rot=<0 0 0.2806 -0.9598>}
 ring_03_l {parent=ring_02_l, loc=<2.4152 0.114 0.0199>, rot=<0 0 0.1189 -0.9929>}
 pinky_metacarpal_l {parent=hand_l, loc=<1.7616 -0.9753 3.3814>, rot=<-0.0397 0.0814 -0.069 -0.9935>}
 pinky_01_l {parent=pinky_metacarpal_l, loc=<7.1709 -0.4831 -0.1094>, rot=<-0.0654 -0.0035 0.2242 -0.9723>}
 pinky_02_l {parent=pinky_01_l, loc=<2.8274 0.0893 -0.1043>, rot=<0.0037 0.0011 0.2854 -0.9584>}
 pinky_03_l {parent=pinky_02_l, loc=<2.2046 0.0307 -0.0078>, rot=<-0.0002 0.0098 0.1445 -0.9895>}
 lowerarm_twist_02_l {parent=lowerarm_l, loc=<9.6199 0 0>, rot=<0 0 0 -1>}
 lowerarm_twist_01_l {parent=lowerarm_l, loc=<17.8549 0 0>, rot=<0 0 0 -1>}
 upperarm_twist_01_l {parent=upperarm_l, loc=<0.0041 0 0>, rot=<0 0 0 1>}
 deform_upperarm_bk_base_l {parent=upperarm_twist_01_l, loc=<1.0813 -7.2193 -1.9453>, rot=<0 0 0 1>}
 deform_upperarm_bk_l {parent=deform_upperarm_bk_base_l, loc=<0 0 0>, rot=<0 0 0 1>}
 deform_upperarm_fr_base_l {parent=upperarm_twist_01_l, loc=<-0.3691 7.0506 -2.765>, rot=<0 0 0 1>}
 deform_upperarm_fr_l {parent=deform_upperarm_fr_base_l, loc=<0 0 0>, rot=<0 0 0 1>}
 deform_upperarm_up_base_l {parent=upperarm_twist_01_l, loc=<-3.2216 -1.1391 3.3311>, rot=<0 0 0 1>}
 deform_upperarm_up_l {parent=deform_upperarm_up_base_l, loc=<0 0 0>, rot=<0 0 0 1>}
 upperarm_twist_02_l {parent=upperarm_l, loc=<13.7941 0 0>, rot=<0 0 0 1>}
 dyn_L_aboveSleeve_1 {parent=upperarm_l, loc=<4.7545 -0.3328 4.4888>, rot=<-0.0086 0.0624 -0.0173 -0.9979>}
 dyn_L_aboveSleeve_2 {parent=dyn_L_aboveSleeve_1, loc=<12.517 0 0>, rot=<0 0 0 1>}
 dyn_L_underSleeve_1 {parent=upperarm_l, loc=<9.8633 0.0509 -8.917>, rot=<-0.6654 0.0157 -0.0371 -0.7454>}
 dyn_L_underSleeve_2 {parent=dyn_L_underSleeve_1, loc=<9.2061 0 0>, rot=<0 0 0 1>}
 dyn_L_backSleeve_1 {parent=upperarm_l, loc=<5.9249 -6.3273 -3.1983>, rot=<-0.1008 0.0606 -0.023 -0.9928>}
 dyn_L_backSleeve_2 {parent=dyn_L_backSleeve_1, loc=<12.517 0 0>, rot=<0 0 0 1>}
 dyn_L_sleeveScoot {parent=upperarm_l, loc=<15.9501 7.1125 -0.2336>, rot=<-0.0079 0.0202 -0.0176 -0.9996>}
 deform_shoulderpad_l_MSTR {parent=clavicle_l, loc=<17.5307 -0.6684 9.4191>, rot=<-0.0398 -0.2811 0.0083 -0.9588>}
 deform_shoulderpad_l {parent=deform_shoulderpad_l_MSTR, loc=<0 0 0>, rot=<0 0 0 1>}
 deform_shoulderpad_front_l {parent=deform_shoulderpad_l, loc=<0.0656 6.6655 -3.9061>, rot=<0.5039 0.0909 -0.0534 -0.8573>}
 deform_shoulderpad_rear_l {parent=deform_shoulderpad_l, loc=<0.7639 -5.6946 -1.7016>, rot=<-0.4158 0.048 -0.1042 -0.9022>}
 neck_01 {parent=spine_05, loc=<11.3138 0.1529 0>, rot=<0 0 -0.2003 -0.9797>}
 neck_02 {parent=neck_01, loc=<4.5627 0 0>, rot=<0 0 0.0504 -0.9987>}
 head {parent=neck_02, loc=<5.3914 0 0>, rot=<0 0 0.115 -0.9934>}
 clavicle_r {parent=spine_05, loc=<4.3575 -0.1466 2.3223>, rot=<0.7289 0.0465 -0.6829 0.009>}
 upperarm_r {parent=clavicle_r, loc=<-15.5111 0 0>, rot=<0.0226 -0.3234 -0.0088 -0.9459>}
 lowerarm_r {parent=upperarm_r, loc=<-26.6514 0 0>, rot=<0.0391 -0.0099 -0.2453 -0.9686>}
 hand_r {parent=lowerarm_r, loc=<-22.3491 0 0>, rot=<-0.7081 -0.0507 0.0975 -0.6975>}
 thumb_01_r {parent=hand_r, loc=<-1.3118 0.3191 2.459>, rot=<0.6546 -0.4042 0.0664 -0.6354>}
 thumb_02_r {parent=thumb_01_r, loc=<-5.9336 0.0448 -0.0401>, rot=<-0.0037 0.0213 0.298 -0.9543>}
 thumb_03_r {parent=thumb_02_r, loc=<-2.2223 -0.3682 0.0907>, rot=<0.0118 -0.013 0.0066 -0.9998>}
 index_metacarpal_r {parent=hand_r, loc=<-1.6535 -0.0183 2.96>, rot=<0.0158 -0.0124 -0.0173 -0.9996>}
 index_01_r {parent=index_metacarpal_r, loc=<-8.3008 -0.0357 0.0524>, rot=<-0.0044 0.0485 0.0924 -0.9945>}
 index_02_r {parent=index_01_r, loc=<-3.3587 0.3004 -0.0452>, rot=<-0.002 0.0027 0.3225 -0.9466>}
 index_03_r {parent=index_02_r, loc=<-2.5578 0.123 0.0252>, rot=<-0.002 0.0067 0.2558 -0.9667>}
 middle_metacarpal_r {parent=hand_r, loc=<-2.1856 0.1725 0.5984>, rot=<0.006 0.0168 -0.0329 -0.9993>}
 middle_01_r {parent=middle_metacarpal_r, loc=<-7.3275 0.2447 -0.0129>, rot=<-0.0208 0.0042 0.0858 -0.9961>}
 middle_02_r {parent=middle_01_r, loc=<-4.0123 -0.1401 -0.1175>, rot=<-0.0218 0.0258 0.3512 -0.9357>}
 middle_03_r {parent=middle_02_r, loc=<-2.6077 -0.0534 -0.1373>, rot=<0.0082 0.0188 0.1708 -0.9851>}
 ring_metacarpal_r {parent=hand_r, loc=<-2.2074 0.3625 -1.3638>, rot=<-0.0284 0.0722 -0.0458 -0.9959>}
 ring_01_r {parent=ring_metacarpal_r, loc=<-7.0295 0.6197 0.1446>, rot=<-0.0489 -0.0252 0.1892 -0.9804>}
 ring_02_r {parent=ring_01_r, loc=<-3.6174 -0.4647 0.1027>, rot=<0 0 0.2806 -0.9598>}
 ring_03_r {parent=ring_02_r, loc=<-2.4152 -0.114 -0.0199>, rot=<0 0 0.1189 -0.9929>}
 pinky_metacarpal_r {parent=hand_r, loc=<-1.7616 0.9753 -3.3814>, rot=<-0.0397 0.0814 -0.069 -0.9935>}
 pinky_01_r {parent=pinky_metacarpal_r, loc=<-7.1709 0.4831 0.1094>, rot=<-0.0654 -0.0035 0.2242 -0.9723>}
 pinky_02_r {parent=pinky_01_r, loc=<-2.8274 -0.0893 0.1043>, rot=<0.0037 0.0011 0.2854 -0.9584>}
 pinky_03_r {parent=pinky_02_r, loc=<-2.2046 -0.0307 0.0078>, rot=<-0.0002 0.0098 0.1445 -0.9895>}
 lowerarm_twist_02_r {parent=lowerarm_r, loc=<-9.6199 0 0>, rot=<0 0 0 -1>}
 lowerarm_twist_01_r {parent=lowerarm_r, loc=<-17.8549 0 0>, rot=<0 0 0 -1>}
 upperarm_twist_01_r {parent=upperarm_r, loc=<-0.0041 0 0>, rot=<0 0 0 -1>}
 deform_upperarm_bk_base_r {parent=upperarm_twist_01_r, loc=<-1.0814 7.2192 1.9456>, rot=<0 0 0 1>}
 deform_upperarm_bk_r {parent=deform_upperarm_bk_base_r, loc=<0 0 0>, rot=<0 0 0 1>}
 deform_upperarm_fr_base_r {parent=upperarm_twist_01_r, loc=<0.3689 -7.0506 2.7652>, rot=<0 0 0 1>}
 deform_upperarm_fr_r {parent=deform_upperarm_fr_base_r, loc=<0 0 0>, rot=<0 0 0 1>}
 deform_upperarm_up_base_r {parent=upperarm_twist_01_r, loc=<3.2214 1.1391 -3.3308>, rot=<0 0 0 1>}
 deform_upperarm_up_r {parent=deform_upperarm_up_base_r, loc=<0 0 0>, rot=<0 0 0 -1>}
 upperarm_twist_02_r {parent=upperarm_r, loc=<-13.7941 0 0>, rot=<0 0 0 -1>}
 dyn_R_aboveSleeve_1 {parent=upperarm_r, loc=<-4.7543 0.3328 -4.489>, rot=<-0.0086 0.0624 -0.0173 -0.9979>}
 dyn_R_aboveSleeve_2 {parent=dyn_R_aboveSleeve_1, loc=<-12.517 0 0>, rot=<0 0 0 1>}
 dyn_R_underSleeve_1 {parent=upperarm_r, loc=<-9.8633 -0.0509 8.9169>, rot=<-0.6654 0.0157 -0.0371 -0.7454>}
 dyn_R_underSleeve_2 {parent=dyn_R_underSleeve_1, loc=<-9.206 -0.0001 0>, rot=<0 0 0 -1>}
 dyn_R_backSleeve_1 {parent=upperarm_r, loc=<-5.9249 6.3273 3.1982>, rot=<-0.1008 0.0606 -0.023 -0.9928>}
 dyn_R_backSleeve_2 {parent=dyn_R_backSleeve_1, loc=<-12.517 0 0.0001>, rot=<0 0 0 1>}
 dyn_R_sleeveScoot {parent=upperarm_r, loc=<-15.9503 -7.1125 0.2339>, rot=<-0.0079 0.0202 -0.0176 -0.9996>}
 deform_shoulderpad_r_MSTR {parent=clavicle_r, loc=<-17.5306 0.6684 -9.4196>, rot=<-0.0398 -0.2811 0.0083 -0.9588>}
 deform_shoulderpad_r {parent=deform_shoulderpad_r_MSTR, loc=<0 0 0>, rot=<0 0 0 1>}
 deform_shoulderpad_front_r {parent=deform_shoulderpad_r, loc=<-0.0657 -6.6655 3.9062>, rot=<0.5039 0.0909 -0.0534 -0.8573>}
 deform_shoulderpad_rear_r {parent=deform_shoulderpad_r, loc=<-0.7644 5.6946 1.7023>, rot=<-0.4158 0.048 -0.1042 -0.9022>}
 dyn_hoodWobble {parent=spine_05, loc=<12.2473 -0.5746 0>, rot=<0 0 0.4728 -0.8812>}
 dyn_hoodFix {parent=dyn_hoodWobble, loc=<7.9832 0 0>, rot=<0 0 0 1>}
 dyn_backStrap_1_MSTR {parent=spine_02, loc=<6.999 -14.1213 0>, rot=<0.0424 -0.9991 0 0>}
 dyn_backStrap_1 {parent=dyn_backStrap_1_MSTR, loc=<0 0 0>, rot=<0 0 0 1>}
 dyn_backStrap_2 {parent=dyn_backStrap_1, loc=<5.2805 0 0>, rot=<0 0 0 1>}
 thigh_l {parent=pelvis, loc=<-2.5693 -0.0109 -7.7501>, rot=<0.0012 0.0572 0.0155 -0.9982>}
 calf_l {parent=thigh_l, loc=<-39.0484 0 0>, rot=<0.0006 -0.0161 -0.0391 -0.9991>}
 foot_l {parent=calf_l, loc=<-42.561 0.1077 0.4292>, rot=<0.0418 -0.0455 0.0343 -0.9975>}
 ball_l {parent=foot_l, loc=<-6.1599 13.8264 0.6806>, rot=<0.0001 -0.0001 -0.7063 -0.7079>}
 calf_twist_02_l {parent=calf_l, loc=<-18.8064 -0.0967 0.0397>, rot=<0.0237 0.0095 -0.0035 -0.9997>}
 calf_twist_01_l {parent=calf_l, loc=<-35.813 -0.0967 0.0397>, rot=<0.0237 0.0095 -0.0035 -0.9997>}
 deform_knee_in_base_l {parent=calf_l, loc=<1.298 -5.2957 -0.0496>, rot=<0 0 0 -1>}
 deform_knee_in_l {parent=deform_knee_in_base_l, loc=<0 0 0>, rot=<0 0 0 1>}
 deform_knee_out_base_l {parent=calf_l, loc=<0 4 0>, rot=<0 0 0 -1>}
 deform_knee_out_l {parent=deform_knee_out_base_l, loc=<0 0 0>, rot=<0 0 0 1>}
 thigh_twist_01_l {parent=thigh_l, loc=<-21.6303 0 0>, rot=<0.0013 0 0 -1>}
 thigh_r {parent=pelvis, loc=<-2.5693 -0.0109 7.7501>, rot=<0.0572 -0.0012 0.9982 0.0155>}
 calf_r {parent=thigh_r, loc=<39.0484 0 0>, rot=<0.0006 -0.0161 -0.0391 -0.9991>}
 foot_r {parent=calf_r, loc=<42.561 -0.1077 -0.4292>, rot=<0.0418 -0.0455 0.0343 -0.9975>}
 ball_r {parent=foot_r, loc=<6.1599 -13.8264 -0.6806>, rot=<0.0001 -0.0001 -0.7063 -0.7079>}
 calf_twist_02_r {parent=calf_r, loc=<18.8064 0.0967 -0.0397>, rot=<0.0237 0.0095 -0.0035 -0.9997>}
 calf_twist_01_r {parent=calf_r, loc=<35.813 0.0967 -0.0397>, rot=<0.0237 0.0095 -0.0035 -0.9997>}
 deform_knee_in_base_r {parent=calf_r, loc=<-1.298 5.2957 0.0496>, rot=<0 0 0 -1>}
 deform_knee_in_r {parent=deform_knee_in_base_r, loc=<0 0 0>, rot=<0 0 0 1>}
 deform_knee_out_base_r {parent=calf_r, loc=<0 -4 0>, rot=<0 0 0 -1>}
 deform_knee_out_r {parent=deform_knee_out_base_r, loc=<0 0 0>, rot=<0 0 0 1>}
 thigh_twist_01_r {parent=thigh_r, loc=<21.6303 0 0>, rot=<0.0013 0 0 -1>}
 dyn_boneBelt_center_1 {parent=thigh_r, loc=<1.4732 1.2511 5.9492>, rot=<0 0 0 1>}
 dyn_boneBelt_front_1 {parent=thigh_r, loc=<2.4159 -7.8133 6.7436>, rot=<0 0 0 1>}
 dyn_boneBelt_front_2 {parent=dyn_boneBelt_front_1, loc=<11.7197 1.1878 0.6093>, rot=<-0.3481 -0.0378 -0.2271 -0.9088>}
 dyn_boneBelt_back_2_ctrl {controlType=EULER_TRANSFORM, parent=thigh_r, type=Control, loc=<0 0 0>, rot=<0 0 0 1>}
 deform_glute_base_l {parent=pelvis, loc=<-8.8064 -8.8222 -8.4287>, rot=<0.0275 0.0501 -0.4472 -0.8926>}
 deform_glute_l {parent=deform_glute_base_l, loc=<0 0 0>, rot=<0 0 0 1>}
 deform_glute_base_r {parent=pelvis, loc=<-8.8062 -8.8222 8.4287>, rot=<0.0501 -0.0275 0.8926 -0.4472>}
 deform_glute_r {parent=deform_glute_base_r, loc=<0 0 0>, rot=<0 0 0 1>}
 dyn_boneBelt_back_1 {parent=pelvis, loc=<-2.8279 -10.3269 12.9245>, rot=<0.0572 -0.0012 0.9982 0.0155>}
 dyn_boneBelt_back_2 {parent=dyn_boneBelt_back_1, loc=<10.0373 -2.2785 1.9672>, rot=<0.1489 0.0827 0.1393 -0.9755>}
 dyn_tshirt_front_MSTR {parent=pelvis, loc=<9.2565 10.1716 0.0003>, rot=<0.0004 0 0.0709 -0.9975>}
 dyn_tshirt_front_01 {parent=dyn_tshirt_front_MSTR, loc=<0 0 0>, rot=<0 0 0 1>}
 dyn_tshirt_front_02 {parent=dyn_tshirt_front_01, loc=<-8.1701 0 0>, rot=<0 0 0 1>}
 dyn_tshirt_front_03 {parent=dyn_tshirt_front_02, loc=<-9.7822 0 0>, rot=<0 0 0 -1>}
 dyn_tshirt_rear_l_MSTR {parent=pelvis, loc=<8.909 -9.924 -12.1474>, rot=<0.9296 0 0.0049 -0.3686>}
 dyn_tshirt_rear_l_01 {parent=dyn_tshirt_rear_l_MSTR, loc=<0 0 0>, rot=<0 0 0 1>}
 dyn_tshirt_rear_l_02 {parent=dyn_tshirt_rear_l_01, loc=<-7.9921 0 0>, rot=<0 0 0 1>}
 dyn_tshirt_rear_l_03 {parent=dyn_tshirt_rear_l_02, loc=<-9.7822 0 0>, rot=<0 0 0 -1>}
 dyn_tshirt_side_l_MSTR {parent=pelvis, loc=<8.3948 -0.3579 -15.5668>, rot=<0.7132 0.0217 0.0563 -0.6984>}
 dyn_tshirt_side_l_01 {parent=dyn_tshirt_side_l_MSTR, loc=<0 0 0>, rot=<0 0 0 1>}
 dyn_tshirt_side_l_02 {parent=dyn_tshirt_side_l_01, loc=<-8.0632 0 0>, rot=<0 0 0 1>}
 dyn_tshirt_side_l_03 {parent=dyn_tshirt_side_l_02, loc=<-9.7822 0 0>, rot=<0 0 0 -1>}
 dyn_tshirt_front_l_MSTR {parent=pelvis, loc=<9.5018 8.5353 -10.0514>, rot=<0.1953 0.0229 0.0565 -0.9789>}
 dyn_tshirt_front_l_01 {parent=dyn_tshirt_front_l_MSTR, loc=<0 0 0>, rot=<0 0 0 1>}
 dyn_tshirt_front_l_02 {parent=dyn_tshirt_front_l_01, loc=<-8.1132 0 0>, rot=<0 0 0 1>}
 dyn_tshirt_front_l_03 {parent=dyn_tshirt_front_l_02, loc=<-9.7822 0 0>, rot=<0 0 0 -1>}
 dyn_tshirt_rear_MSTR {parent=pelvis, loc=<8.5533 -12.6976 0.0129>, rot=<1 0.0008 0 -0.0067>}
 dyn_tshirt_rear_01 {parent=dyn_tshirt_rear_MSTR, loc=<0 0 0>, rot=<0 0 0 1>}
 dyn_tshirt_rear_02 {parent=dyn_tshirt_rear_01, loc=<-8.0519 0 0>, rot=<0 0 0 1>}
 dyn_tshirt_rear_03 {parent=dyn_tshirt_rear_02, loc=<-9.7822 0 0>, rot=<0 0 0 -1>}
 dyn_tshirt_rear_r_MSTR {parent=pelvis, loc=<8.9091 -9.924 12.1474>, rot=<0 -0.9296 0.3686 0.0049>}
 dyn_tshirt_rear_r_01 {parent=dyn_tshirt_rear_r_MSTR, loc=<0 0 0>, rot=<0 0 0 1>}
 dyn_tshirt_rear_r_02 {parent=dyn_tshirt_rear_r_01, loc=<7.9922 0 0>, rot=<0 0 0 -1>}
 dyn_tshirt_rear_r_03 {parent=dyn_tshirt_rear_r_02, loc=<9.7822 0 0>, rot=<0 0 0 -1>}
 dyn_tshirt_side_r_MSTR {parent=pelvis, loc=<8.3944 -0.3579 15.5668>, rot=<0.0217 -0.7132 0.6984 0.0563>}
 dyn_tshirt_side_r_01 {parent=dyn_tshirt_side_r_MSTR, loc=<0 0 0>, rot=<0 0 0 1>}
 dyn_tshirt_side_r_02 {parent=dyn_tshirt_side_r_01, loc=<8.0629 0 0>, rot=<0 0 0 1>}
 dyn_tshirt_side_r_03 {parent=dyn_tshirt_side_r_02, loc=<9.7822 0 0>, rot=<0 0 0 -1>}
 dyn_tshirt_front_r_MSTR {parent=pelvis, loc=<9.5017 8.5353 10.0514>, rot=<0.0229 -0.1953 0.9789 0.0565>}
 dyn_tshirt_front_r_01 {parent=dyn_tshirt_front_r_MSTR, loc=<0 0 0>, rot=<0 0 0 1>}
 dyn_tshirt_front_r_02 {parent=dyn_tshirt_front_r_01, loc=<8.1131 0 0>, rot=<0 0 0 1>}
 dyn_tshirt_front_r_03 {parent=dyn_tshirt_front_r_02, loc=<9.7822 0 0>, rot=<0 0 0 -1>}
]
variables [
 L lowerarm bend {type=double, default=0}
 L thighForward alpha {type=double, default=0}
 L thighInward reverseAlpha {type=double, default=0}
 R lowerarm bend {type=double, default=0}
 R thighForward alpha {type=double, default=0}
 R thighInward reverseAlpha {type=double, default=0}
 isNotPlayingEmoteInFrontend {type=double, default=1}
 neckBack alpha {type=double, default=0}
 spr_arm_l_fwd {type=float, default=0}
 spr_arm_r_fwd {type=float, default=0}
 spr_clavicle_l_raise {type=float, default=0}
 spr_clavicle_r_raise {type=float, default=0}
]
