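# Physics asset for the left sleeve of the male medium LemonCartCleaner body cosmetic, providing capsule bodies for its dynamic sleeve bones.
# Three two-bone chains (above, back and under) hang from the upperarm_l capsule, with tight angular limit values along each chain.
# Cross constraints with free swing and twist tie the back chain to both neighbors at each level, and its second bone to the upper arm.
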PhysicsAsset /BRCosmetics/Characters/Player/Male/Medium/Bodies/M_MED_LemonCartCleaner/Meshes/M_MED_LemonCartCleaner_LeftSleeve_Physics
bodies {
 dyn_L_aboveSleeve_1 [{shapeType=Capsule, length=16, radius=5, shapeName=dyn_L_aboveSleeve_1_capsule, center=<4 0 3>, rotation=<rot 0 0 90>}]
 dyn_L_aboveSleeve_2 [{shapeType=Capsule, length=16, radius=5, shapeName=dyn_L_aboveSleeve_2_capsule, center=<4 0 3>, rotation=<rot 0 0 90>}]
 dyn_L_backSleeve_1 [{shapeType=Capsule, length=15.5212, radius=5.4788, shapeName=dyn_L_backSleeve_1_capsule, center=<4 -4 0>, rotation=<rot 0 90 0>}]
 dyn_L_backSleeve_2 [{shapeType=Capsule, length=15.7494, radius=5.2506, shapeName=dyn_L_backSleeve_2_capsule, center=<4 -4 0>, rotation=<rot 0 90 0>}]
 dyn_L_underSleeve_1 [{shapeType=Capsule, length=16, radius=4, shapeName=dyn_L_underSleeve_1_capsule, center=<4 -4 0>, rotation=<rot 0 90 0>}]
 dyn_L_underSleeve_2 [{shapeType=Capsule, length=16, radius=5, shapeName=dyn_L_underSleeve_2_capsule, center=<4 -4 0>, rotation=<rot 0 90 0>}]
 upperarm_l [{shapeType=Capsule, length=20, radius=6, center=<14 1 -2>, rotation=<rot 0.0001 90.0001 100>}]
}
constraints [
 {bone1Name=dyn_L_aboveSleeve_1, bone2Name=upperarm_l, swing1LimitDegrees=5, swing2LimitDegrees=5, twistLimitDegrees=5}
 {bone1Name=dyn_L_aboveSleeve_2, bone2Name=dyn_L_aboveSleeve_1, swing1LimitDegrees=5, swing2LimitDegrees=5, twistLimitDegrees=5}
 {bone1Name=dyn_L_backSleeve_1, bone2Name=upperarm_l, swing1LimitDegrees=5, swing2LimitDegrees=5, twistLimitDegrees=5}
 {bone1Name=dyn_L_backSleeve_2, bone2Name=dyn_L_backSleeve_1, swing1LimitDegrees=5, swing2LimitDegrees=5, twistLimitDegrees=5}
 {bone1Name=dyn_L_underSleeve_1, bone2Name=upperarm_l, swing1LimitDegrees=5, swing2LimitDegrees=5, twistLimitDegrees=5}
 {bone1Name=dyn_L_underSleeve_2, bone2Name=dyn_L_underSleeve_1, swing1LimitDegrees=5, swing2LimitDegrees=5, twistLimitDegrees=5}
 {
  bone1Name=dyn_L_backSleeve_1
  bone2Name=dyn_L_aboveSleeve_1
  swing1LimitDegrees=45
  swing1Motion=Free
  swing2LimitDegrees=45
  swing2Motion=Free
  twistLimitDegrees=45
  twistMotion=Free
 }
 {
  bone1Name=dyn_L_backSleeve_2
  bone2Name=dyn_L_aboveSleeve_2
  swing1LimitDegrees=45
  swing1Motion=Free
  swing2LimitDegrees=45
  swing2Motion=Free
  twistLimitDegrees=45
  twistMotion=Free
 }
 {
  bone1Name=dyn_L_backSleeve_1
  bone2Name=dyn_L_underSleeve_1
  swing1LimitDegrees=45
  swing1Motion=Free
  swing2LimitDegrees=45
  swing2Motion=Free
  twistLimitDegrees=45
  twistMotion=Free
 }
 {
  bone1Name=dyn_L_backSleeve_2
  bone2Name=dyn_L_underSleeve_2
  swing1LimitDegrees=45
  swing1Motion=Free
  swing2LimitDegrees=45
  swing2Motion=Free
  twistLimitDegrees=45
  twistMotion=Free
 }
 {
  bone1Name=dyn_L_backSleeve_2
  bone2Name=upperarm_l
  swing1LimitDegrees=45
  swing1Motion=Free
  swing2LimitDegrees=45
  swing2Motion=Free
  twistLimitDegrees=45
  twistMotion=Free
 }
]
