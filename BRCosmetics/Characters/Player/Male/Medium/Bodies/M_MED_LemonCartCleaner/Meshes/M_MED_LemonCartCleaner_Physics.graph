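# Physics asset for the LemonCartCleaner medium male player body, covering secondary motion of its shirt, hood, belt and back straps.
# Capsules on pelvis, spine and thighs form the body that the dyn_ bone capsules attach to.
# Eight three-link shirt chains hang from tiny MSTR anchors on the pelvis with tight swing and twist limits, with free-rotation links tying front-side tips to side tips.
# Hood, straps and belt attach to spine, pelvis or right thigh.
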PhysicsAsset /BRCosmetics/Characters/Player/Male/Medium/Bodies/M_MED_LemonCartCleaner/Meshes/M_MED_LemonCartCleaner_Physics
bodies {
 dyn_backStrap_1 [{shapeType=Capsule, length=6.4375, radius=3.5312, shapeName=dyn_backStrap_1_capsule, center=<4 0 0>, rotation=<rot 0 97.8001 90>}]
 dyn_backStrap_1_MSTR [{shapeType=Capsule, length=0.505, radius=0.505, shapeName=dyn_backStrap_1_MSTR_capsule, rotation=<rot 0 -90 0>}]
 dyn_backStrap_2 [{shapeType=Capsule, length=4.75, radius=3.4062, shapeName=dyn_backStrap_2_capsule, center=<5.0635 1 0>, rotation=<rot 0 90 90>}]
 dyn_boneBelt_back_1 [{shapeType=Capsule, length=6, radius=1.75, shapeName=dyn_boneBelt_back_1_capsule, center=<4 -0.5 0.5>, rotation=<rot 2.5759 80.3339 104.7822>}]
 dyn_boneBelt_back_2 [{shapeType=Capsule, length=7.7812, radius=1.75, shapeName=dyn_boneBelt_back_2_capsule, center=<4 0 0>, rotation=<rot 0 90 90>}]
 dyn_boneBelt_center_1 [
  {
   shapeType=Capsule
   length=5.291
   radius=8.9034
   shapeName=dyn_boneBelt_center_1_capsule
   center=<17.1202 -0.8804 -4.2093>
   rotation=<rot 0.5556 -179.9193 -88.1678>
  }
  {
   shapeType=Capsule
   length=8
   radius=6.3125
   shapeName=dyn_boneBelt_center_1_capsule
   center=<5.4123 -0.0285 -1.5099>
   rotation=<rot 0 0 -88.918>
  }
 ]
 dyn_hoodWobble [{shapeType=Capsule, length=8, radius=5, shapeName=dyn_hoodWobble_capsule, center=<7 0 0>}]
 dyn_tshirt_front_01 [{shapeType=Capsule, length=17.3157, radius=5.6843, shapeName=dyn_tshirt_front_01_capsule, center=<-6 4.75 0>}]
 dyn_tshirt_front_02 [{shapeType=Capsule, length=16.5181, radius=6.4819, shapeName=dyn_tshirt_front_02_capsule, center=<-6 4.5 0>}]
 dyn_tshirt_front_03 [{shapeType=Capsule, length=13.4375, radius=6.125, shapeName=dyn_tshirt_front_03_capsule, center=<-3.6617 4.1684 0>}]
 dyn_tshirt_front_MSTR [{shapeType=Capsule, length=0.505, radius=0.505, shapeName=dyn_tshirt_front_MSTR_capsule, rotation=<rot 0 -90 0>}]
 dyn_tshirt_front_l_01 [{shapeType=Capsule, length=17.2646, radius=5.7354, shapeName=dyn_tshirt_front_l_01_capsule, center=<-6 3.5 0>}]
 dyn_tshirt_front_l_02 [{shapeType=Capsule, length=16.5687, radius=6.4313, shapeName=dyn_tshirt_front_l_02_capsule, center=<-6 5 0>}]
 dyn_tshirt_front_l_03 [{shapeType=Capsule, length=16.7812, radius=6.7188, shapeName=dyn_tshirt_front_l_03_capsule, center=<-4.0855 4.1915 0>}]
 dyn_tshirt_front_l_MSTR [{shapeType=Capsule, length=0.505, radius=0.505, shapeName=dyn_tshirt_front_l_MSTR_capsule, rotation=<rot 0 -90 0>}]
 dyn_tshirt_front_r_01 [{shapeType=Capsule, length=17.2646, radius=5.7354, shapeName=dyn_tshirt_front_r_01_capsule, center=<6 -5 0>}]
 dyn_tshirt_front_r_02 [{shapeType=Capsule, length=16.975, radius=6.025, shapeName=dyn_tshirt_front_r_02_capsule, center=<6 -3.5 0>}]
 dyn_tshirt_front_r_03 [{shapeType=Capsule, length=18.0839, radius=5.4161, shapeName=dyn_tshirt_front_r_03_capsule, center=<6 -3.5 0>}]
 dyn_tshirt_front_r_MSTR [{shapeType=Capsule, length=0.505, radius=0.505, shapeName=dyn_tshirt_front_r_MSTR_capsule, rotation=<rot 0 -90 0>}]
 dyn_tshirt_rear_01 [{shapeType=Capsule, length=18, radius=5, shapeName=dyn_tshirt_rear_01_capsule, center=<-6 3.5 0>}]
 dyn_tshirt_rear_02 [{shapeType=Capsule, length=18, radius=5, shapeName=dyn_tshirt_rear_02_capsule, center=<-6 3.5 0>}]
 dyn_tshirt_rear_03 [{shapeType=Capsule, length=18, radius=5, shapeName=dyn_tshirt_rear_03_capsule, center=<-6 3.5 0>}]
 dyn_tshirt_rear_MSTR [{shapeType=Capsule, length=0.505, radius=0.505, shapeName=dyn_tshirt_rear_MSTR_capsule, rotation=<rot 0 -90 0>}]
 dyn_tshirt_rear_l_01 [{shapeType=Capsule, length=18, radius=5, shapeName=dyn_tshirt_rear_l_01_capsule, center=<-6 3.5 0>}]
 dyn_tshirt_rear_l_02 [{shapeType=Capsule, length=18, radius=5, shapeName=dyn_tshirt_rear_l_02_capsule, center=<-6 3.5 0>}]
 dyn_tshirt_rear_l_03 [{shapeType=Capsule, length=18, radius=5, shapeName=dyn_tshirt_rear_l_03_capsule, center=<-6 3.5 0>}]
 dyn_tshirt_rear_l_MSTR [{shapeType=Capsule, length=0.505, radius=0.505, shapeName=dyn_tshirt_rear_l_MSTR_capsule, rotation=<rot 0 -90 0>}]
 dyn_tshirt_rear_r_01 [{shapeType=Capsule, length=18, radius=5, shapeName=dyn_tshirt_rear_r_01_capsule, center=<6 -3.5 0>}]
 dyn_tshirt_rear_r_02 [{shapeType=Capsule, length=18, radius=5, shapeName=dyn_tshirt_rear_r_02_capsule, center=<6 -3.5 0>}]
 dyn_tshirt_rear_r_03 [{shapeType=Capsule, length=18, radius=5, shapeName=dyn_tshirt_rear_r_03_capsule, center=<6 -3.5 0>}]
 dyn_tshirt_rear_r_MSTR [{shapeType=Capsule, length=0.505, radius=0.505, shapeName=dyn_tshirt_rear_r_MSTR_capsule, rotation=<rot 0 -90 0>}]
 dyn_tshirt_side_l_01 [{shapeType=Capsule, length=18, radius=5, shapeName=dyn_tshirt_side_l_01_capsule, center=<-6 4.7631 0>}]
 dyn_tshirt_side_l_02 [{shapeType=Capsule, length=18, radius=5, shapeName=dyn_tshirt_side_l_02_capsule, center=<-6 6 0>}]
 dyn_tshirt_side_l_03 [{shapeType=Capsule, length=18, radius=5, shapeName=dyn_tshirt_side_l_03_capsule, center=<-6 4.5 0>}]
 dyn_tshirt_side_l_MSTR [{shapeType=Capsule, length=0.505, radius=0.505, shapeName=dyn_tshirt_side_l_MSTR_capsule, rotation=<rot 0 -90 0>}]
 dyn_tshirt_side_r_01 [{shapeType=Capsule, length=18, radius=5, shapeName=dyn_tshirt_side_l_01_capsule, center=<6 -4.7631 0>, rotation=<rot 0 0 -180>}]
 dyn_tshirt_side_r_02 [{shapeType=Capsule, length=18, radius=5, shapeName=dyn_tshirt_side_l_02_capsule, center=<6 -6 0>, rotation=<rot 0 0 -180>}]
 dyn_tshirt_side_r_03 [{shapeType=Capsule, length=18, radius=5, shapeName=dyn_tshirt_side_l_03_capsule, center=<6 -4.5 0>, rotation=<rot 0 0 -180>}]
 dyn_tshirt_side_r_MSTR [{shapeType=Capsule, length=0.505, radius=0.505, shapeName=dyn_tshirt_side_r_MSTR_capsule, rotation=<rot 0 -90 0>}]
 pelvis [
  {shapeType=Capsule, length=10, radius=11, center=<0 -1 0>}
  {shapeType=Capsule, length=10, radius=12, shapeName=pelvis_capsule, center=<-10.3882 -1 0>}
 ]
 spine_01 [{shapeType=Capsule, length=11.3125, radius=10.6875, center=<0 -1 0>}]
 spine_02 [{shapeType=Capsule, length=12, radius=10, shapeName=spine_02_capsule, center=<0 -0.5 0>}]
 spine_03 [{shapeType=Capsule, length=12, radius=10, center=<0 -1 0>}]
 spine_04 [{shapeType=Capsule, length=12, radius=10, center=<0 -1 0>}]
 spine_05 [{shapeType=Capsule, length=12, radius=10, center=<0 -1 0>}]
 thigh_l [{shapeType=Capsule, length=25, radius=10, center=<-18 -0.5 2>, rotation=<rot 0 89.9998 85>}]
 thigh_r [{shapeType=Capsule, length=25, radius=10, center=<18 0.5 -2>, rotation=<rot 0 89.9999 85>}]
}
constraints [
 {bone1Name=spine_01, bone2Name=pelvis, swing1LimitDegrees=45, swing2LimitDegrees=45, twistLimitDegrees=45}
 {bone1Name=spine_03, bone2Name=spine_02, swing1LimitDegrees=45, swing2LimitDegrees=45, twistLimitDegrees=45}
 {bone1Name=spine_04, bone2Name=spine_03, swing1LimitDegrees=45, swing2LimitDegrees=45, twistLimitDegrees=45}
 {bone1Name=spine_05, bone2Name=spine_04, swing1LimitDegrees=45, swing2LimitDegrees=45, twistLimitDegrees=45}
 {bone1Name=thigh_l, bone2Name=pelvis, swing1LimitDegrees=45, swing2LimitDegrees=45, twistLimitDegrees=45}
 {bone1Name=thigh_r, bone2Name=pelvis, swing1LimitDegrees=45, swing2LimitDegrees=45, twistLimitDegrees=45}
 {bone1Name=dyn_hoodWobble, bone2Name=spine_05, swing1LimitDegrees=5, swing2LimitDegrees=5, twistLimitDegrees=1}
 {bone1Name=dyn_tshirt_front_01, bone2Name=dyn_tshirt_front_MSTR, swing1LimitDegrees=15, swing2LimitDegrees=15, twistLimitDegrees=1}
 {bone1Name=dyn_tshirt_front_02, bone2Name=dyn_tshirt_front_01, swing1LimitDegrees=15, swing2LimitDegrees=15, twistLimitDegrees=1}
 {bone1Name=dyn_tshirt_front_03, bone2Name=dyn_tshirt_front_02, swing1LimitDegrees=15, swing2LimitDegrees=15, twistLimitDegrees=1}
 {bone1Name=dyn_tshirt_front_l_01, bone2Name=dyn_tshirt_front_l_MSTR, swing1LimitDegrees=15, swing2LimitDegrees=3, twistLimitDegrees=1}
 {bone1Name=dyn_tshirt_front_l_02, bone2Name=dyn_tshirt_front_l_01, swing1LimitDegrees=15, swing2LimitDegrees=3, twistLimitDegrees=1}
 {bone1Name=dyn_tshirt_front_l_03, bone2Name=dyn_tshirt_front_l_02, swing1LimitDegrees=15, swing2LimitDegrees=3, twistLimitDegrees=1}
 {bone1Name=dyn_tshirt_front_r_01, bone2Name=dyn_tshirt_front_r_MSTR, swing1LimitDegrees=15, swing2LimitDegrees=10, twistLimitDegrees=1}
 {bone1Name=dyn_tshirt_front_r_02, bone2Name=dyn_tshirt_front_r_01, swing1LimitDegrees=15, swing2LimitDegrees=10, twistLimitDegrees=1}
 {bone1Name=dyn_tshirt_front_r_03, bone2Name=dyn_tshirt_front_r_02, swing1LimitDegrees=15, swing2LimitDegrees=10, twistLimitDegrees=1}
 {bone1Name=dyn_tshirt_rear_01, bone2Name=dyn_tshirt_rear_MSTR, swing1LimitDegrees=15, swing2LimitDegrees=15, twistLimitDegrees=1}
 {bone1Name=dyn_tshirt_rear_02, bone2Name=dyn_tshirt_rear_01, swing1LimitDegrees=15, swing2LimitDegrees=15, twistLimitDegrees=1}
 {bone1Name=dyn_tshirt_rear_03, bone2Name=dyn_tshirt_rear_02, swing1LimitDegrees=15, swing2LimitDegrees=15, twistLimitDegrees=1}
 {bone1Name=dyn_tshirt_rear_l_01, bone2Name=dyn_tshirt_rear_l_MSTR, swing1LimitDegrees=15, swing2LimitDegrees=15, twistLimitDegrees=1}
 {bone1Name=dyn_tshirt_rear_l_02, bone2Name=dyn_tshirt_rear_l_01, swing1LimitDegrees=15, swing2LimitDegrees=15, twistLimitDegrees=1}
 {bone1Name=dyn_tshirt_rear_l_03, bone2Name=dyn_tshirt_rear_l_02, swing1LimitDegrees=15, swing2LimitDegrees=15, twistLimitDegrees=1}
 {bone1Name=dyn_tshirt_rear_r_01, bone2Name=dyn_tshirt_rear_r_MSTR, swing1LimitDegrees=15, swing2LimitDegrees=15, twistLimitDegrees=1}
 {bone1Name=dyn_tshirt_rear_r_02, bone2Name=dyn_tshirt_rear_r_01, swing1LimitDegrees=15, swing2LimitDegrees=15, twistLimitDegrees=1}
 {bone1Name=dyn_tshirt_rear_r_03, bone2Name=dyn_tshirt_rear_r_02, swing1LimitDegrees=15, swing2LimitDegrees=15, twistLimitDegrees=1}
 {bone1Name=dyn_tshirt_side_l_01, bone2Name=dyn_tshirt_side_l_MSTR, swing1LimitDegrees=15, swing2LimitDegrees=15, twistLimitDegrees=1}
 {bone1Name=dyn_tshirt_side_l_02, bone2Name=dyn_tshirt_side_l_01, swing1LimitDegrees=15, swing2LimitDegrees=15, twistLimitDegrees=1}
 {bone1Name=dyn_tshirt_side_l_03, bone2Name=dyn_tshirt_side_l_02, swing1LimitDegrees=15, swing2LimitDegrees=15, twistLimitDegrees=1}
 {bone1Name=dyn_tshirt_side_r_01, bone2Name=dyn_tshirt_side_r_MSTR, swing1LimitDegrees=15, swing2LimitDegrees=15, twistLimitDegrees=1}
 {bone1Name=dyn_tshirt_side_r_02, bone2Name=dyn_tshirt_side_r_01, swing1LimitDegrees=15, swing2LimitDegrees=15, twistLimitDegrees=1}
 {bone1Name=dyn_tshirt_side_r_03, bone2Name=dyn_tshirt_side_r_02, swing1LimitDegrees=15, swing2LimitDegrees=15, twistLimitDegrees=1}
 {bone1Name=dyn_boneBelt_back_2, bone2Name=dyn_boneBelt_back_1, swing1LimitDegrees=5, swing2LimitDegrees=13, twistLimitDegrees=1}
 {bone1Name=dyn_boneBelt_back_1, bone2Name=pelvis, swing1LimitDegrees=5, swing2LimitDegrees=13, twistLimitDegrees=1}
 {bone1Name=dyn_backStrap_1, bone2Name=dyn_backStrap_1_MSTR, swing1LimitDegrees=26, swing2LimitDegrees=26, twistLimitDegrees=8}
 {bone1Name=dyn_backStrap_2, bone2Name=dyn_backStrap_1, swing1LimitDegrees=26, swing2LimitDegrees=26, twistLimitDegrees=8}
 {
  bone1Name=spine_02
  bone2Name=spine_01
  swing1LimitDegrees=45
  swing1Motion=Free
  swing2LimitDegrees=45
  swing2Motion=Free
  twistLimitDegrees=45
  twistMotion=Free
 }
 {
  bone1Name=dyn_tshirt_front_l_MSTR
  bone2Name=pelvis
  swing1LimitDegrees=45
  swing1Motion=Free
  swing2LimitDegrees=45
  swing2Motion=Free
  twistLimitDegrees=45
  twistMotion=Free
 }
 {
  bone1Name=dyn_tshirt_front_r_MSTR
  bone2Name=pelvis
  swing1LimitDegrees=45
  swing1Motion=Free
  swing2LimitDegrees=45
  swing2Motion=Free
  twistLimitDegrees=45
  twistMotion=Free
 }
 {
  bone1Name=dyn_tshirt_rear_MSTR
  bone2Name=pelvis
  swing1LimitDegrees=45
  swing1Motion=Free
  swing2LimitDegrees=45
  swing2Motion=Free
  twistLimitDegrees=45
  twistMotion=Free
 }
 {
  bone1Name=dyn_tshirt_rear_l_MSTR
  bone2Name=pelvis
  swing1LimitDegrees=45
  swing1Motion=Free
  swing2LimitDegrees=45
  swing2Motion=Free
  twistLimitDegrees=45
  twistMotion=Free
 }
 {
  bone1Name=dyn_tshirt_rear_r_MSTR
  bone2Name=pelvis
  swing1LimitDegrees=45
  swing1Motion=Free
  swing2LimitDegrees=45
  swing2Motion=Free
  twistLimitDegrees=45
  twistMotion=Free
 }
 {
  bone1Name=dyn_tshirt_side_l_MSTR
  bone2Name=pelvis
  swing1LimitDegrees=45
  swing1Motion=Free
  swing2LimitDegrees=45
  swing2Motion=Free
  twistLimitDegrees=45
  twistMotion=Free
 }
 {
  bone1Name=dyn_tshirt_side_r_MSTR
  bone2Name=pelvis
  swing1LimitDegrees=45
  swing1Motion=Free
  swing2LimitDegrees=45
  swing2Motion=Free
  twistLimitDegrees=45
  twistMotion=Free
 }
 {
  bone1Name=dyn_tshirt_front_MSTR
  bone2Name=pelvis
  swing1LimitDegrees=45
  swing1Motion=Free
  swing2LimitDegrees=45
  swing2Motion=Free
  twistLimitDegrees=45
  twistMotion=Free
 }
 {
  bone1Name=dyn_tshirt_front_l_03
  bone2Name=dyn_tshirt_side_l_03
  swing1LimitDegrees=45
  swing1Motion=Free
  swing2LimitDegrees=45
  swing2Motion=Free
  twistLimitDegrees=45
  twistMotion=Free
 }
 {
  bone1Name=dyn_tshirt_front_r_03
  bone2Name=dyn_tshirt_side_r_03
  swing1LimitDegrees=45
  swing1Motion=Free
  swing2LimitDegrees=45
  swing2Motion=Free
  twistLimitDegrees=45
  twistMotion=Free
 }
 {
  bone1Name=dyn_boneBelt_center_1
  bone2Name=thigh_r
  swing1LimitDegrees=45
  swing1Motion=Free
  swing2LimitDegrees=45
  swing2Motion=Free
  twistLimitDegrees=45
  twistMotion=Free
 }
 {
  bone1Name=dyn_backStrap_1_MSTR
  bone2Name=spine_02
  swing1LimitDegrees=45
  swing1Motion=Free
  swing2LimitDegrees=45
  swing2Motion=Free
  twistLimitDegrees=45
  twistMotion=Free
 }
]
