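# Post-deformation AnimBlueprint for the LemonCartCleaner male medium body cosmetic, adding secondary motion to its clothing bones.
# It copies the parent mesh's pose through a Control Rig, then caches it for two physics paths, selected at output by bRigidBodyEnabled.
# The RigidBody path chains body and per-sleeve physics assets; the AnimDynamics alternative springs sleeves, hood and back strap.
# Both constrain the back strap to spine and shirt bones; separate T-shirt and belt AnimDynamics chains appear unconnected.
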AnimBlueprint /BRCosmetics/Characters/Player/Male/Medium/Bodies/M_MED_LemonCartCleaner/Meshes/M_MED_LemonCartCleaner_AnimBP
target_skeleton=&/M_MED_LemonCartCleaner_Skeleton
graphs [
 AnimGraph {nodes [
   Anim Dynamics - Bone: dyn_tshirt_side_l_03!(SimulationSpace:AnimPhysSimSpaceType Enum, ComponentPose, Alpha=1) -> (Pose) {
    angularDampingOverride=1
    angularSpringConstant=12000
    bAngularSpring=true
    bOverrideAngularDamping=true
    bUseGravityOverride=true
    gravityScale=0.35
    simulationSpace=RootRelative
    alphaBoolBlend {blendOption=Linear}
    boundBone {boneName=dyn_tshirt_side_l_03}
    componentAppliedLinearAccClamp=<2000 2000 2000>
    componentLinearAccScale=<1 1 1>
    componentLinearVelScale=<1 1 1>
    gravityOverride {z=-1250}
    physicsBodyDefinitions [{
      boundBone {boneName=dyn_tshirt_side_l_03}
      boxExtents=<19 19 19>
      constraintSetup {angularLimitsMax=<15 5 0>, angularLimitsMin=<-15 -5 -20>, angularTarget=<1 0 0>}
      localJointOffset=<-4 0 0>
     }]
   }
   Anim Dynamics - Bone: dyn_tshirt_front_l_03!(SimulationSpace:AnimPhysSimSpaceType Enum, ComponentPose, Alpha=1) -> (Pose) {
    angularDampingOverride=1
    angularSpringConstant=12000
    bAngularSpring=true
    bOverrideAngularDamping=true
    bUseGravityOverride=true
    gravityScale=0.35
    simulationSpace=RootRelative
    alphaBoolBlend {blendOption=Linear}
    boundBone {boneName=dyn_tshirt_front_l_03}
    componentAppliedLinearAccClamp=<2000 2000 2000>
    componentLinearAccScale=<1 1 1>
    componentLinearVelScale=<1 1 1>
    gravityOverride {z=-1500}
    physicsBodyDefinitions [{
      boundBone {boneName=dyn_tshirt_front_l_03}
      boxExtents=<26 26 26>
      constraintSetup {angularLimitsMax=<0 15 0>, angularLimitsMin=<0 -15 -20>, angularTarget=<1 0 0>}
      localJointOffset=<-4 0 0>
     }]
   }
   n1 = Anim Dynamics - Bone: dyn_tshirt_front_01!(
    SimulationSpace:AnimPhysSimSpaceType Enum
    ComponentPose=Reroute Node(InputPin:Component Space Pose Link Structure) -> (OutputPin:Component Space Pose Link Structure)
    Alpha=1
   ) -> (Pose) {
    angularDampingOverride=1
    angularSpringConstant=12000
    bAngularSpring=true
    bOverrideAngularDamping=true
    bUseGravityOverride=true
    gravityScale=0.35
    simulationSpace=RootRelative
    alphaBoolBlend {blendOption=Linear}
    boundBone {boneName=dyn_tshirt_front_01}
    componentAppliedLinearAccClamp=<2000 2000 2000>
    componentLinearAccScale=<1 1 1>
    componentLinearVelScale=<1 1 1>
    gravityOverride {z=-980}
    physicsBodyDefinitions [{
      boundBone {boneName=dyn_tshirt_front_01}
      boxExtents=<19 19 19>
      constraintSetup {angularLimitsMax=<0 10 0>, angularLimitsMin=<0 -10 -20>, angularTarget=<1 0 0>}
      localJointOffset=<-4 0 0>
     }]
   }
   n2 = Anim Dynamics - Bone: dyn_tshirt_front_02!(SimulationSpace:AnimPhysSimSpaceType Enum, ComponentPose=@n1, Alpha=1) -> (Pose) {
    angularDampingOverride=1
    angularSpringConstant=12000
    bAngularSpring=true
    bOverrideAngularDamping=true
    bUseGravityOverride=true
    gravityScale=0.35
    simulationSpace=RootRelative
    alphaBoolBlend {blendOption=Linear}
    boundBone {boneName=dyn_tshirt_front_02}
    componentAppliedLinearAccClamp=<2000 2000 2000>
    componentLinearAccScale=<1 1 1>
    componentLinearVelScale=<1 1 1>
    gravityOverride {z=-980}
    physicsBodyDefinitions [{
      boundBone {boneName=dyn_tshirt_front_02}
      boxExtents=<19 19 19>
      constraintSetup {angularLimitsMax=<0 10 0>, angularLimitsMin=<0 -10 -20>, angularTarget=<1 0 0>}
      localJointOffset=<-4 0 0>
     }]
   }
   n3 = Anim Dynamics - Bone: dyn_tshirt_front_03!(SimulationSpace:AnimPhysSimSpaceType Enum, ComponentPose=@n2, Alpha=1) -> (Pose) {
    angularDampingOverride=1
    angularSpringConstant=12000
    bAngularSpring=true
    bOverrideAngularDamping=true
    bUseGravityOverride=true
    gravityScale=0.35
    simulationSpace=RootRelative
    alphaBoolBlend {blendOption=Linear}
    boundBone {boneName=dyn_tshirt_front_03}
    componentAppliedLinearAccClamp=<2000 2000 2000>
    componentLinearAccScale=<1 1 1>
    componentLinearVelScale=<1 1 1>
    gravityOverride {z=-980}
    physicsBodyDefinitions [{
      boundBone {boneName=dyn_tshirt_front_03}
      boxExtents=<19 19 19>
      constraintSetup {angularLimitsMax=<0 10 0>, angularLimitsMin=<0 -10 -20>, angularTarget=<1 0 0>}
      localJointOffset=<-4 0 0>
     }]
   }
   n4 = Anim Dynamics - Bone: dyn_tshirt_rear_01!(SimulationSpace:AnimPhysSimSpaceType Enum, ComponentPose=@n3, Alpha=1) -> (Pose) {
    angularDampingOverride=1
    angularSpringConstant=12000
    bAngularSpring=true
    bOverrideAngularDamping=true
    bUseGravityOverride=true
    gravityScale=0.35
    simulationSpace=RootRelative
    alphaBoolBlend {blendOption=Linear}
    boundBone {boneName=dyn_tshirt_rear_01}
    componentAppliedLinearAccClamp=<2000 2000 2000>
    componentLinearAccScale=<1 1 1>
    componentLinearVelScale=<1 1 1>
    gravityOverride {z=-980}
    physicsBodyDefinitions [{
      boundBone {boneName=dyn_tshirt_rear_01}
      boxExtents=<23 23 23>
      constraintSetup {angularLimitsMax=<15 15 0>, angularLimitsMin=<-15 -15 -20>, angularTarget=<1 0 0>}
      localJointOffset=<-4 0 0>
     }]
   }
   n5 = Anim Dynamics - Bone: dyn_tshirt_rear_02!(SimulationSpace:AnimPhysSimSpaceType Enum, ComponentPose=@n4, Alpha=1) -> (Pose) {
    angularDampingOverride=1
    angularSpringConstant=12000
    bAngularSpring=true
    bOverrideAngularDamping=true
    bUseGravityOverride=true
    gravityScale=0.35
    simulationSpace=RootRelative
    alphaBoolBlend {blendOption=Linear}
    boundBone {boneName=dyn_tshirt_rear_02}
    componentAppliedLinearAccClamp=<2000 2000 2000>
    componentLinearAccScale=<1 1 1>
    componentLinearVelScale=<1 1 1>
    gravityOverride {z=-980}
    physicsBodyDefinitions [{
      boundBone {boneName=dyn_tshirt_rear_02}
      boxExtents=<23 23 23>
      constraintSetup {angularLimitsMax=<15 15 0>, angularLimitsMin=<-15 -15 -20>, angularTarget=<1 0 0>}
      localJointOffset=<-4 0 0>
     }]
   }
   n6 = Anim Dynamics - Bone: dyn_tshirt_rear_03!(SimulationSpace:AnimPhysSimSpaceType Enum, ComponentPose=@n5, Alpha=1) -> (Pose) {
    angularDampingOverride=1
    angularSpringConstant=12000
    bAngularSpring=true
    bOverrideAngularDamping=true
    bUseGravityOverride=true
    gravityScale=0.35
    simulationSpace=RootRelative
    alphaBoolBlend {blendOption=Linear}
    boundBone {boneName=dyn_tshirt_rear_03}
    componentAppliedLinearAccClamp=<2000 2000 2000>
    componentLinearAccScale=<1 1 1>
    componentLinearVelScale=<1 1 1>
    gravityOverride {z=-980}
    physicsBodyDefinitions [{
      boundBone {boneName=dyn_tshirt_rear_03}
      boxExtents=<23 23 23>
      constraintSetup {angularLimitsMax=<15 15 0>, angularLimitsMin=<-15 -15 -20>, angularTarget=<1 0 0>}
      localJointOffset=<-4 0 0>
     }]
   }
   n7 = Anim Dynamics - Bone: dyn_tshirt_front_l_01!(SimulationSpace:AnimPhysSimSpaceType Enum, ComponentPose=@n6, Alpha=1) -> (Pose) {
    angularDampingOverride=1
    angularSpringConstant=12000
    bAngularSpring=true
    bOverrideAngularDamping=true
    bUseGravityOverride=true
    gravityScale=0.35
    simulationSpace=RootRelative
    alphaBoolBlend {blendOption=Linear}
    boundBone {boneName=dyn_tshirt_front_l_01}
    componentAppliedLinearAccClamp=<2000 2000 2000>
    componentLinearAccScale=<1 1 1>
    componentLinearVelScale=<1 1 1>
    gravityOverride {z=-1500}
    physicsBodyDefinitions [{
      boundBone {boneName=dyn_tshirt_front_l_01}
      boxExtents=<26 26 26>
      constraintSetup {angularLimitsMax=<0 15 0>, angularLimitsMin=<0 -15 -20>, angularTarget=<1 0 0>}
      localJointOffset=<-4 0 0>
     }]
   }
   n8 = Anim Dynamics - Bone: dyn_tshirt_front_l_02!(SimulationSpace:AnimPhysSimSpaceType Enum, ComponentPose=@n7, Alpha=1) -> (Pose) {
    angularDampingOverride=1
    angularSpringConstant=12000
    bAngularSpring=true
    bOverrideAngularDamping=true
    bUseGravityOverride=true
    gravityScale=0.35
    simulationSpace=RootRelative
    alphaBoolBlend {blendOption=Linear}
    boundBone {boneName=dyn_tshirt_front_l_02}
    componentAppliedLinearAccClamp=<2000 2000 2000>
    componentLinearAccScale=<1 1 1>
    componentLinearVelScale=<1 1 1>
    gravityOverride {z=-1500}
    physicsBodyDefinitions [{
      boundBone {boneName=dyn_tshirt_front_l_02}
      boxExtents=<26 26 26>
      constraintSetup {angularLimitsMax=<0 15 0>, angularLimitsMin=<0 -15 -20>, angularTarget=<1 0 0>}
      localJointOffset=<-4 0 0>
     }]
   }
   n9 = Anim Dynamics - Bone: dyn_tshirt_side_l_01!(SimulationSpace:AnimPhysSimSpaceType Enum, ComponentPose=@n8, Alpha=1) -> (Pose) {
    angularDampingOverride=1
    angularSpringConstant=12000
    bAngularSpring=true
    bOverrideAngularDamping=true
    bUseGravityOverride=true
    gravityScale=0.35
    simulationSpace=RootRelative
    alphaBoolBlend {blendOption=Linear}
    boundBone {boneName=dyn_tshirt_side_l_01}
    componentAppliedLinearAccClamp=<2000 2000 2000>
    componentLinearAccScale=<1 1 1>
    componentLinearVelScale=<1 1 1>
    gravityOverride {z=-1250}
    physicsBodyDefinitions [{
      boundBone {boneName=dyn_tshirt_side_l_01}
      boxExtents=<19 19 19>
      constraintSetup {angularLimitsMax=<15 5 0>, angularLimitsMin=<-15 -5 -20>, angularTarget=<1 0 0>}
      localJointOffset=<-4 0 0>
     }]
   }
   n10 = Anim Dynamics - Bone: dyn_tshirt_side_l_02!(SimulationSpace:AnimPhysSimSpaceType Enum, ComponentPose=@n9, Alpha=1) -> (Pose) {
    angularDampingOverride=1
    angularSpringConstant=12000
    bAngularSpring=true
    bOverrideAngularDamping=true
    bUseGravityOverride=true
    gravityScale=0.35
    simulationSpace=RootRelative
    alphaBoolBlend {blendOption=Linear}
    boundBone {boneName=dyn_tshirt_side_l_02}
    componentAppliedLinearAccClamp=<2000 2000 2000>
    componentLinearAccScale=<1 1 1>
    componentLinearVelScale=<1 1 1>
    gravityOverride {z=-1250}
    physicsBodyDefinitions [{
      boundBone {boneName=dyn_tshirt_side_l_02}
      boxExtents=<19 19 19>
      constraintSetup {angularLimitsMax=<15 5 0>, angularLimitsMin=<-15 -5 -20>, angularTarget=<1 0 0>}
      localJointOffset=<-4 0 0>
     }]
   }
   n11 = Anim Dynamics - Bone: dyn_tshirt_rear_l_01!(SimulationSpace:AnimPhysSimSpaceType Enum, ComponentPose=@n10, Alpha=1) -> (Pose) {
    angularDampingOverride=1
    angularSpringConstant=12000
    bAngularSpring=true
    bOverrideAngularDamping=true
    bUseGravityOverride=true
    gravityScale=0.35
    simulationSpace=RootRelative
    alphaBoolBlend {blendOption=Linear}
    boundBone {boneName=dyn_tshirt_rear_l_01}
    componentAppliedLinearAccClamp=<2000 2000 2000>
    componentLinearAccScale=<1 1 1>
    componentLinearVelScale=<1 1 1>
    gravityOverride {z=-980}
    physicsBodyDefinitions [{
      boundBone {boneName=dyn_tshirt_rear_l_01}
      boxExtents=<32 32 32>
      constraintSetup {angularLimitsMax=<15 15 0>, angularLimitsMin=<-15 -15 -20>, angularTarget=<1 0 0>}
      localJointOffset=<-4 0 0>
     }]
   }
   n12 = Anim Dynamics - Bone: dyn_tshirt_rear_l_02!(SimulationSpace:AnimPhysSimSpaceType Enum, ComponentPose=@n11, Alpha=1) -> (Pose) {
    angularDampingOverride=1
    angularSpringConstant=12000
    bAngularSpring=true
    bOverrideAngularDamping=true
    bUseGravityOverride=true
    gravityScale=0.35
    simulationSpace=RootRelative
    alphaBoolBlend {blendOption=Linear}
    boundBone {boneName=dyn_tshirt_rear_l_02}
    componentAppliedLinearAccClamp=<2000 2000 2000>
    componentLinearAccScale=<1 1 1>
    componentLinearVelScale=<1 1 1>
    gravityOverride {z=-980}
    physicsBodyDefinitions [{
      boundBone {boneName=dyn_tshirt_rear_l_02}
      boxExtents=<32 32 32>
      constraintSetup {angularLimitsMax=<15 15 0>, angularLimitsMin=<-15 -15 -20>, angularTarget=<1 0 0>}
      localJointOffset=<-4 0 0>
     }]
   }
   n13 = Anim Dynamics - Bone: dyn_tshirt_rear_l_03!(SimulationSpace:AnimPhysSimSpaceType Enum, ComponentPose=@n12, Alpha=1) -> (Pose) {
    angularDampingOverride=1
    angularSpringConstant=12000
    bAngularSpring=true
    bOverrideAngularDamping=true
    bUseGravityOverride=true
    gravityScale=0.35
    simulationSpace=RootRelative
    alphaBoolBlend {blendOption=Linear}
    boundBone {boneName=dyn_tshirt_rear_l_03}
    componentAppliedLinearAccClamp=<2000 2000 2000>
    componentLinearAccScale=<1 1 1>
    componentLinearVelScale=<1 1 1>
    gravityOverride {z=-980}
    physicsBodyDefinitions [{
      boundBone {boneName=dyn_tshirt_rear_l_03}
      boxExtents=<32 32 32>
      constraintSetup {angularLimitsMax=<15 15 0>, angularLimitsMin=<-15 -15 -20>, angularTarget=<1 0 0>}
      localJointOffset=<-4 0 0>
     }]
   }
   n14 = Anim Dynamics - Bone: dyn_tshirt_front_r_01!(SimulationSpace:AnimPhysSimSpaceType Enum, ComponentPose=@n13, Alpha=1) -> (Pose) {
    angularDampingOverride=1
    angularSpringConstant=12000
    bAngularSpring=true
    bOverrideAngularDamping=true
    bUseGravityOverride=true
    gravityScale=0.35
    simulationSpace=RootRelative
    alphaBoolBlend {blendOption=Linear}
    boundBone {boneName=dyn_tshirt_front_r_01}
    componentAppliedLinearAccClamp=<2000 2000 2000>
    componentLinearAccScale=<1 1 1>
    componentLinearVelScale=<1 1 1>
    gravityOverride {z=-1200}
    physicsBodyDefinitions [{
      boundBone {boneName=dyn_tshirt_front_r_01}
      boxExtents=<26 26 26>
      constraintSetup {angularLimitsMax=<0 0 0>, angularLimitsMin=<0 -15 -20>, angularTarget=<1 0 0>}
      localJointOffset=<4 0 0>
     }]
   }
   n15 = Anim Dynamics - Bone: dyn_tshirt_front_r_02!(SimulationSpace:AnimPhysSimSpaceType Enum, ComponentPose=@n14, Alpha=1) -> (Pose) {
    angularDampingOverride=1
    angularSpringConstant=12000
    bAngularSpring=true
    bOverrideAngularDamping=true
    bUseGravityOverride=true
    gravityScale=0.35
    simulationSpace=RootRelative
    alphaBoolBlend {blendOption=Linear}
    boundBone {boneName=dyn_tshirt_front_r_02}
    componentAppliedLinearAccClamp=<2000 2000 2000>
    componentLinearAccScale=<1 1 1>
    componentLinearVelScale=<1 1 1>
    gravityOverride {z=-1200}
    physicsBodyDefinitions [{
      boundBone {boneName=dyn_tshirt_front_r_02}
      boxExtents=<26 26 26>
      constraintSetup {angularLimitsMax=<0 0 0>, angularLimitsMin=<0 -15 -20>, angularTarget=<1 0 0>}
      localJointOffset=<4 0 0>
     }]
   }
   n16 = Anim Dynamics - Bone: dyn_tshirt_side_r_01!(SimulationSpace:AnimPhysSimSpaceType Enum, ComponentPose=@n15, Alpha=1) -> (Pose) {
    angularDampingOverride=1
    angularSpringConstant=12000
    bAngularSpring=true
    bOverrideAngularDamping=true
    bUseGravityOverride=true
    gravityScale=0.35
    simulationSpace=RootRelative
    alphaBoolBlend {blendOption=Linear}
    boundBone {boneName=dyn_tshirt_side_r_01}
    componentAppliedLinearAccClamp=<2000 2000 2000>
    componentLinearAccScale=<1 1 1>
    componentLinearVelScale=<1 1 1>
    gravityOverride {z=-1100}
    physicsBodyDefinitions [{
      boundBone {boneName=dyn_tshirt_side_r_01}
      boxExtents=<23 23 23>
      constraintSetup {angularLimitsMax=<0 5 0>, angularLimitsMin=<0 -5 -20>, angularTarget=<1 0 0>}
      localJointOffset=<4 0 0>
     }]
   }
   n17 = Anim Dynamics - Bone: dyn_tshirt_side_r_02!(SimulationSpace:AnimPhysSimSpaceType Enum, ComponentPose=@n16, Alpha=1) -> (Pose) {
    angularDampingOverride=1
    angularSpringConstant=12000
    bAngularSpring=true
    bOverrideAngularDamping=true
    bUseGravityOverride=true
    gravityScale=0.35
    simulationSpace=RootRelative
    alphaBoolBlend {blendOption=Linear}
    boundBone {boneName=dyn_tshirt_side_r_02}
    componentAppliedLinearAccClamp=<2000 2000 2000>
    componentLinearAccScale=<1 1 1>
    componentLinearVelScale=<1 1 1>
    gravityOverride {z=-1100}
    physicsBodyDefinitions [{
      boundBone {boneName=dyn_tshirt_side_r_02}
      boxExtents=<23 23 23>
      constraintSetup {angularLimitsMax=<0 5 0>, angularLimitsMin=<0 -5 -20>, angularTarget=<1 0 0>}
      localJointOffset=<4 0 0>
     }]
   }
   n18 = Anim Dynamics - Bone: dyn_tshirt_rear_r_01!(SimulationSpace:AnimPhysSimSpaceType Enum, ComponentPose=@n17, Alpha=1) -> (Pose) {
    angularDampingOverride=1
    angularSpringConstant=12000
    bAngularSpring=true
    bOverrideAngularDamping=true
    bUseGravityOverride=true
    gravityScale=0.35
    simulationSpace=RootRelative
    alphaBoolBlend {blendOption=Linear}
    boundBone {boneName=dyn_tshirt_rear_r_01}
    componentAppliedLinearAccClamp=<2000 2000 2000>
    componentLinearAccScale=<1 1 1>
    componentLinearVelScale=<1 1 1>
    gravityOverride {z=-980}
    physicsBodyDefinitions [{
      boundBone {boneName=dyn_tshirt_rear_r_01}
      boxExtents=<28 28 28>
      constraintSetup {angularLimitsMax=<15 15 0>, angularLimitsMin=<-15 0 -20>, angularTarget=<1 0 0>}
      localJointOffset=<4 0 0>
     }]
   }
   Anim Dynamics - Bone: dyn_tshirt_rear_r_02!(SimulationSpace:AnimPhysSimSpaceType Enum, ComponentPose=@n18, Alpha=1) -> (Pose) {
    angularDampingOverride=1
    angularSpringConstant=12000
    bAngularSpring=true
    bOverrideAngularDamping=true
    bUseGravityOverride=true
    gravityScale=0.35
    simulationSpace=RootRelative
    alphaBoolBlend {blendOption=Linear}
    boundBone {boneName=dyn_tshirt_rear_r_02}
    componentAppliedLinearAccClamp=<2000 2000 2000>
    componentLinearAccScale=<1 1 1>
    componentLinearVelScale=<1 1 1>
    gravityOverride {z=-980}
    physicsBodyDefinitions [{
      boundBone {boneName=dyn_tshirt_rear_r_02}
      boxExtents=<28 28 28>
      constraintSetup {angularLimitsMax=<15 15 0>, angularLimitsMin=<-15 0 -20>, angularTarget=<1 0 0>}
      localJointOffset=<4 0 0>
     }]
   }
   Anim Dynamics - Bone: dyn_tshirt_rear_r_03!(SimulationSpace:AnimPhysSimSpaceType Enum, ComponentPose, Alpha=1) -> (Pose) {
    angularDampingOverride=1
    angularSpringConstant=12000
    bAngularSpring=true
    bOverrideAngularDamping=true
    bUseGravityOverride=true
    gravityScale=0.35
    simulationSpace=RootRelative
    alphaBoolBlend {blendOption=Linear}
    boundBone {boneName=dyn_tshirt_rear_r_03}
    componentAppliedLinearAccClamp=<2000 2000 2000>
    componentLinearAccScale=<1 1 1>
    componentLinearVelScale=<1 1 1>
    gravityOverride {z=-980}
    physicsBodyDefinitions [{
      boundBone {boneName=dyn_tshirt_rear_r_03}
      boxExtents=<28 28 28>
      constraintSetup {angularLimitsMax=<15 15 0>, angularLimitsMin=<-15 0 -20>, angularTarget=<1 0 0>}
      localJointOffset=<4 0 0>
     }]
   }
   Anim Dynamics - Bone: dyn_tshirt_side_r_03!(SimulationSpace:AnimPhysSimSpaceType Enum, ComponentPose, Alpha=1) -> (Pose) {
    angularDampingOverride=1
    angularSpringConstant=12000
    bAngularSpring=true
    bOverrideAngularDamping=true
    bUseGravityOverride=true
    gravityScale=0.35
    simulationSpace=RootRelative
    alphaBoolBlend {blendOption=Linear}
    boundBone {boneName=dyn_tshirt_side_r_03}
    componentAppliedLinearAccClamp=<2000 2000 2000>
    componentLinearAccScale=<1 1 1>
    componentLinearVelScale=<1 1 1>
    gravityOverride {z=-1100}
    physicsBodyDefinitions [{
      boundBone {boneName=dyn_tshirt_side_r_03}
      boxExtents=<23 23 23>
      constraintSetup {angularLimitsMax=<0 5 0>, angularLimitsMin=<0 -5 -20>, angularTarget=<1 0 0>}
      localJointOffset=<4 0 0>
     }]
   }
   Anim Dynamics - Bone: dyn_tshirt_front_r_03!(SimulationSpace:AnimPhysSimSpaceType Enum, ComponentPose, Alpha=1) -> (Pose) {
    angularDampingOverride=1
    angularSpringConstant=12000
    bAngularSpring=true
    bOverrideAngularDamping=true
    bUseGravityOverride=true
    gravityScale=0.35
    simulationSpace=RootRelative
    alphaBoolBlend {blendOption=Linear}
    boundBone {boneName=dyn_tshirt_front_r_03}
    componentAppliedLinearAccClamp=<2000 2000 2000>
    componentLinearAccScale=<1 1 1>
    componentLinearVelScale=<1 1 1>
    gravityOverride {z=-1200}
    physicsBodyDefinitions [{
      boundBone {boneName=dyn_tshirt_front_r_03}
      boxExtents=<26 26 26>
      constraintSetup {angularLimitsMax=<0 0 0>, angularLimitsMin=<0 -15 -20>, angularTarget=<1 0 0>}
      localJointOffset=<4 0 0>
     }]
   }
   n19 = Anim Dynamics - Bone: dyn_boneBelt_front_1!(SimulationSpace:AnimPhysSimSpaceType Enum, ComponentPose, Alpha=1) -> (Pose) {
    angularDampingOverride=1
    angularSpringConstant=2000
    bAngularSpring=true
    bOverrideAngularDamping=true
    bUseGravityOverride=true
    gravityScale=0.35
    simulationSpace=RootRelative
    alphaBoolBlend {blendOption=Linear}
    boundBone {boneName=dyn_boneBelt_front_1}
    componentAppliedLinearAccClamp=<2000 2000 2000>
    componentLinearAccScale=<1 1 1>
    componentLinearVelScale=<1 1 1>
    gravityOverride {z=-980}
    physicsBodyDefinitions [{
      boundBone {boneName=dyn_boneBelt_front_1}
      boxExtents=<20 30 30>
      constraintSetup {angularLimitsMax=<0 0 0>, angularLimitsMin=<0 -60 -15>, angularTarget=<1 0 0>}
      localJointOffset=<4 0 0>
     }]
   }
   n20 = Anim Dynamics - Bone: dyn_boneBelt_center_1!(SimulationSpace:AnimPhysSimSpaceType Enum, ComponentPose=@n19, Alpha=1) -> (Pose) {
    angularDampingOverride=1
    angularSpringConstant=2000
    bAngularSpring=true
    bOverrideAngularDamping=true
    bUseGravityOverride=true
    gravityScale=0.35
    simulationSpace=RootRelative
    alphaBoolBlend {blendOption=Linear}
    boundBone {boneName=dyn_boneBelt_center_1}
    componentAppliedLinearAccClamp=<2000 2000 2000>
    componentLinearAccScale=<1 1 1>
    componentLinearVelScale=<1 1 1>
    gravityOverride {z=-980}
    physicsBodyDefinitions [{
      boundBone {boneName=dyn_boneBelt_center_1}
      boxExtents=<20 30 30>
      constraintSetup {angularLimitsMax=<0 0 0>, angularLimitsMin=<0 -60 0>, angularTarget=<1 0 0>}
      localJointOffset=<4 0 0>
     }]
   }
   n21 = Anim Dynamics - Bone: dyn_boneBelt_back_1!(SimulationSpace:AnimPhysSimSpaceType Enum, ComponentPose=@n20, Alpha=1) -> (Pose) {
    angularDampingOverride=1
    angularSpringConstant=2000
    bAngularSpring=true
    bOverrideAngularDamping=true
    bUseGravityOverride=true
    gravityScale=0.35
    simulationSpace=RootRelative
    alphaBoolBlend {blendOption=Linear}
    boundBone {boneName=dyn_boneBelt_back_1}
    componentAppliedLinearAccClamp=<2000 2000 2000>
    componentLinearAccScale=<1 1 1>
    componentLinearVelScale=<1 1 1>
    gravityOverride {z=-980}
    physicsBodyDefinitions [{
      boundBone {boneName=dyn_boneBelt_back_1}
      boxExtents=<20 30 30>
      constraintSetup {angularLimitsMax=<0 0 0>, angularLimitsMin=<0 0 -15>, angularTarget=<1 0 0>}
      localJointOffset=<4 0 0>
     }]
   }
   "Constraint - dyn_boneBelt_center_1 (PP)"!(ComponentPose=@n21, Alpha=1, ConstraintWeights_0=0.5, ConstraintWeights_1:float=0.5) -> (Pose) {
    alphaBoolBlend {blendOption=Linear}
    alphaScaleBiasClamp {scale=0.3333}
    boneToModify {boneName=dyn_boneBelt_center_1}
    constraintSetup [{targetBone {boneName=dyn_boneBelt_front_1}}, {targetBone {boneName=dyn_boneBelt_back_1}}]
    constraintWeights [1,1]
   }
   Output Pose(Result=Blend Poses by bool(BlendPose_0=Use cached pose 'PostRBAN'(), BlendPose_1=Use cached pose 'PostAnimDynamics'(), bActiveValue=Get bRigidBodyEnabled()) {blendType=Linear}) {layerGroup=None}
   n22 = Control Rig!(
    Source=Copy Pose From Mesh() {bUseAttachedParent=true}
    Alpha
    isNotPlayingEmoteInFrontend:double=Get isNotPlayingEmoteInFrontend() -> (isNotPlayingEmoteInFrontend:double)
   ) -> (Pose) {
    bTransferInputCurves=false
    alphaBoolBlend {blendOption=Linear}
    controlRigAssetReference {blueprintRigClass=&/M_MED_LemonCartCleaner_CtrlRig.M_MED_LemonCartCleaner_CtrlRig_C}
   }
   Save cached pose 'PostDeformation'(Pose=@n22) {cacheName=PostDeformation}
   n23 = Anim Dynamics - Bone: dyn_L_aboveSleeve_1!(SimulationSpace:AnimPhysSimSpaceType Enum, ComponentPose=Local To Component(LocalPose=Use cached pose 'PostDeformation'()), Alpha=1) -> (Pose) {
    angularDampingOverride=0.9999
    angularSpringConstant=11000
    bAngularSpring=true
    bOverrideAngularDamping=true
    bUseGravityOverride=true
    gravityScale=0.35
    simulationSpace=RootRelative
    alphaBoolBlend {blendOption=Linear}
    boundBone {boneName=dyn_L_aboveSleeve_1}
    componentAppliedLinearAccClamp=<2000 2000 2000>
    componentLinearAccScale=<1 1 1>
    componentLinearVelScale=<1 1 1>
    gravityOverride {z=-1500}
    physicsBodyDefinitions [{
      boundBone {boneName=dyn_L_aboveSleeve_1}
      boxExtents=<25 25 25>
      constraintSetup {angularLimitsMax=<5 0 5>, angularLimitsMin=<-5 -25 -10>, angularTarget=<1 0 0>}
      localJointOffset=<4 0 0>
     }]
   }
   n24 = Anim Dynamics - Bone: dyn_L_aboveSleeve_2!(SimulationSpace:AnimPhysSimSpaceType Enum, ComponentPose=@n23, Alpha=1) -> (Pose) {
    angularDampingOverride=0.9999
    angularSpringConstant=11000
    bAngularSpring=true
    bOverrideAngularDamping=true
    bUseGravityOverride=true
    gravityScale=0.35
    simulationSpace=RootRelative
    alphaBoolBlend {blendOption=Linear}
    boundBone {boneName=dyn_L_aboveSleeve_2}
    componentAppliedLinearAccClamp=<2000 2000 2000>
    componentLinearAccScale=<1 1 1>
    componentLinearVelScale=<1 1 1>
    gravityOverride {z=-1500}
    physicsBodyDefinitions [{
      boundBone {boneName=dyn_L_aboveSleeve_2}
      boxExtents=<25 25 25>
      constraintSetup {angularLimitsMax=<5 0 5>, angularLimitsMin=<-5 -25 -10>, angularTarget=<1 0 0>}
      localJointOffset=<4 0 0>
     }]
   }
   n25 = Anim Dynamics - Bone: dyn_L_underSleeve_1!(SimulationSpace:AnimPhysSimSpaceType Enum, ComponentPose=@n24, Alpha=1) -> (Pose) {
    angularDampingOverride=0.9999
    angularSpringConstant=11000
    bAngularSpring=true
    bOverrideAngularDamping=true
    bUseGravityOverride=true
    gravityScale=0.35
    simulationSpace=RootRelative
    alphaBoolBlend {blendOption=Linear}
    boundBone {boneName=dyn_L_underSleeve_1}
    componentAppliedLinearAccClamp=<2000 2000 2000>
    componentLinearAccScale=<1 1 1>
    componentLinearVelScale=<1 1 1>
    gravityOverride {z=-1500}
    physicsBodyDefinitions [{
      boundBone {boneName=dyn_L_underSleeve_1}
      boxExtents=<19 19 19>
      constraintSetup {angularLimitsMax=<-5 5 0>, angularLimitsMin=<-5 -5 -10>, angularTarget=<1 0 0>}
      localJointOffset=<4 0 0>
     }]
   }
   n26 = Anim Dynamics - Bone: dyn_L_underSleeve_2!(SimulationSpace:AnimPhysSimSpaceType Enum, ComponentPose=@n25, Alpha=1) -> (Pose) {
    angularDampingOverride=0.9999
    angularSpringConstant=11000
    bAngularSpring=true
    bOverrideAngularDamping=true
    bUseGravityOverride=true
    gravityScale=0.35
    simulationSpace=RootRelative
    alphaBoolBlend {blendOption=Linear}
    boundBone {boneName=dyn_L_underSleeve_2}
    componentAppliedLinearAccClamp=<2000 2000 2000>
    componentLinearAccScale=<1 1 1>
    componentLinearVelScale=<1 1 1>
    gravityOverride {z=-1500}
    physicsBodyDefinitions [{
      boundBone {boneName=dyn_L_underSleeve_2}
      boxExtents=<19 19 19>
      constraintSetup {angularLimitsMax=<-5 5 0>, angularLimitsMin=<-5 -5 -10>, angularTarget=<1 0 0>}
      localJointOffset=<4 0 0>
     }]
   }
   n27 = Anim Dynamics - Bone: dyn_L_backSleeve_1!(SimulationSpace:AnimPhysSimSpaceType Enum, ComponentPose=@n26, Alpha=1) -> (Pose) {
    angularDampingOverride=0.9999
    angularSpringConstant=11000
    bAngularSpring=true
    bOverrideAngularDamping=true
    bUseGravityOverride=true
    gravityScale=0.35
    simulationSpace=RootRelative
    alphaBoolBlend {blendOption=Linear}
    boundBone {boneName=dyn_L_backSleeve_1}
    componentAppliedLinearAccClamp=<2000 2000 2000>
    componentLinearAccScale=<1 1 1>
    componentLinearVelScale=<1 1 1>
    gravityOverride {z=-1500}
    physicsBodyDefinitions [{
      boundBone {boneName=dyn_L_backSleeve_1}
      boxExtents=<19 19 19>
      constraintSetup {angularLimitsMax=<0 10 0>, angularLimitsMin=<0 -10 -15>, angularTarget=<1 0 0>}
      localJointOffset=<4 0 0>
     }]
   }
   n28 = Anim Dynamics - Bone: dyn_L_backSleeve_2!(SimulationSpace:AnimPhysSimSpaceType Enum, ComponentPose=@n27, Alpha=1) -> (Pose) {
    angularDampingOverride=0.9999
    angularSpringConstant=11000
    bAngularSpring=true
    bOverrideAngularDamping=true
    bUseGravityOverride=true
    gravityScale=0.35
    simulationSpace=RootRelative
    alphaBoolBlend {blendOption=Linear}
    boundBone {boneName=dyn_L_backSleeve_2}
    componentAppliedLinearAccClamp=<2000 2000 2000>
    componentLinearAccScale=<1 1 1>
    componentLinearVelScale=<1 1 1>
    gravityOverride {z=-1500}
    physicsBodyDefinitions [{
      boundBone {boneName=dyn_L_backSleeve_2}
      boxExtents=<19 19 19>
      constraintSetup {angularLimitsMax=<0 5 0>, angularLimitsMin=<0 -5 -10>, angularTarget=<1 0 0>}
      localJointOffset=<4 0 0>
     }]
   }
   n29 = Anim Dynamics - Bone: dyn_R_aboveSleeve_1!(SimulationSpace:AnimPhysSimSpaceType Enum, ComponentPose=@n28, Alpha=1) -> (Pose) {
    angularDampingOverride=0.9999
    angularSpringConstant=11000
    bAngularSpring=true
    bOverrideAngularDamping=true
    bUseGravityOverride=true
    gravityScale=0.35
    simulationSpace=RootRelative
    alphaBoolBlend {blendOption=Linear}
    boundBone {boneName=dyn_R_aboveSleeve_1}
    componentAppliedLinearAccClamp=<2000 2000 2000>
    componentLinearAccScale=<1 1 1>
    componentLinearVelScale=<1 1 1>
    gravityOverride {z=-1500}
    physicsBodyDefinitions [{
      boundBone {boneName=dyn_R_aboveSleeve_1}
      boxExtents=<25 25 25>
      constraintSetup {angularLimitsMax=<5 0 5>, angularLimitsMin=<-5 -25 -10>, angularTarget=<1 0 0>}
      localJointOffset=<-4 0 0>
     }]
   }
   n30 = Anim Dynamics - Bone: dyn_R_aboveSleeve_2!(SimulationSpace:AnimPhysSimSpaceType Enum, ComponentPose=@n29, Alpha=1) -> (Pose) {
    angularDampingOverride=0.9999
    angularSpringConstant=11000
    bAngularSpring=true
    bOverrideAngularDamping=true
    bUseGravityOverride=true
    gravityScale=0.35
    simulationSpace=RootRelative
    alphaBoolBlend {blendOption=Linear}
    boundBone {boneName=dyn_R_aboveSleeve_2}
    componentAppliedLinearAccClamp=<2000 2000 2000>
    componentLinearAccScale=<1 1 1>
    componentLinearVelScale=<1 1 1>
    gravityOverride {z=-1500}
    physicsBodyDefinitions [{
      boundBone {boneName=dyn_R_aboveSleeve_2}
      boxExtents=<25 25 25>
      constraintSetup {angularLimitsMax=<5 0 5>, angularLimitsMin=<-5 -25 -10>, angularTarget=<1 0 0>}
      localJointOffset=<-4 0 0>
     }]
   }
   n31 = Anim Dynamics - Bone: dyn_R_underSleeve_1!(SimulationSpace:AnimPhysSimSpaceType Enum, ComponentPose=@n30, Alpha=1) -> (Pose) {
    angularDampingOverride=0.9999
    angularSpringConstant=11000
    bAngularSpring=true
    bOverrideAngularDamping=true
    bUseGravityOverride=true
    gravityScale=0.35
    simulationSpace=RootRelative
    alphaBoolBlend {blendOption=Linear}
    boundBone {boneName=dyn_R_underSleeve_1}
    componentAppliedLinearAccClamp=<2000 2000 2000>
    componentLinearAccScale=<1 1 1>
    componentLinearVelScale=<1 1 1>
    gravityOverride {z=-1500}
    physicsBodyDefinitions [{
      boundBone {boneName=dyn_R_underSleeve_1}
      boxExtents=<19 19 19>
      constraintSetup {angularLimitsMax=<-5 5 0>, angularLimitsMin=<-5 -5 -10>, angularTarget=<1 0 0>}
      localJointOffset=<-4 0 0>
     }]
   }
   n32 = Anim Dynamics - Bone: dyn_R_underSleeve_2!(SimulationSpace:AnimPhysSimSpaceType Enum, ComponentPose=@n31, Alpha=1) -> (Pose) {
    angularDampingOverride=0.9999
    angularSpringConstant=11000
    bAngularSpring=true
    bOverrideAngularDamping=true
    bUseGravityOverride=true
    gravityScale=0.35
    simulationSpace=RootRelative
    alphaBoolBlend {blendOption=Linear}
    boundBone {boneName=dyn_R_underSleeve_2}
    componentAppliedLinearAccClamp=<2000 2000 2000>
    componentLinearAccScale=<1 1 1>
    componentLinearVelScale=<1 1 1>
    gravityOverride {z=-1500}
    physicsBodyDefinitions [{
      boundBone {boneName=dyn_R_underSleeve_2}
      boxExtents=<19 19 19>
      constraintSetup {angularLimitsMax=<-5 5 0>, angularLimitsMin=<-5 -5 -10>, angularTarget=<1 0 0>}
      localJointOffset=<-4 0 0>
     }]
   }
   n33 = Anim Dynamics - Bone: dyn_R_backSleeve_1!(SimulationSpace:AnimPhysSimSpaceType Enum, ComponentPose=@n32, Alpha=1) -> (Pose) {
    angularDampingOverride=0.9999
    angularSpringConstant=11000
    bAngularSpring=true
    bOverrideAngularDamping=true
    bUseGravityOverride=true
    gravityScale=0.35
    simulationSpace=RootRelative
    alphaBoolBlend {blendOption=Linear}
    boundBone {boneName=dyn_R_backSleeve_1}
    componentAppliedLinearAccClamp=<2000 2000 2000>
    componentLinearAccScale=<1 1 1>
    componentLinearVelScale=<1 1 1>
    gravityOverride {z=-1500}
    physicsBodyDefinitions [{
      boundBone {boneName=dyn_R_backSleeve_1}
      boxExtents=<19 19 19>
      constraintSetup {angularLimitsMax=<0 10 0>, angularLimitsMin=<0 -10 -15>, angularTarget=<1 0 0>}
      localJointOffset=<-4 0 0>
     }]
   }
   n34 = Anim Dynamics - Bone: dyn_R_backSleeve_2!(SimulationSpace:AnimPhysSimSpaceType Enum, ComponentPose=@n33, Alpha=1) -> (Pose) {
    angularDampingOverride=0.9999
    angularSpringConstant=11000
    bAngularSpring=true
    bOverrideAngularDamping=true
    bUseGravityOverride=true
    gravityScale=0.35
    simulationSpace=RootRelative
    alphaBoolBlend {blendOption=Linear}
    boundBone {boneName=dyn_R_backSleeve_2}
    componentAppliedLinearAccClamp=<2000 2000 2000>
    componentLinearAccScale=<1 1 1>
    componentLinearVelScale=<1 1 1>
    gravityOverride {z=-1500}
    physicsBodyDefinitions [{
      boundBone {boneName=dyn_R_backSleeve_2}
      boxExtents=<19 19 19>
      constraintSetup {angularLimitsMax=<0 5 0>, angularLimitsMin=<0 -5 -10>, angularTarget=<1 0 0>}
      localJointOffset=<-4 0 0>
     }]
   }
   n35 = Anim Dynamics - Bone: dyn_hoodWobble!(SimulationSpace:AnimPhysSimSpaceType Enum, ComponentPose=@n34, Alpha=1) -> (Pose) {
    angularDampingOverride=1
    angularSpringConstant=10000
    bAngularSpring=true
    bOverrideAngularDamping=true
    bUseGravityOverride=true
    gravityScale=0.35
    simulationSpace=RootRelative
    alphaBoolBlend {blendOption=Linear}
    boundBone {boneName=dyn_hoodWobble}
    componentAppliedLinearAccClamp=<2000 2000 2000>
    componentLinearAccScale=<1 1 1>
    componentLinearVelScale=<1 1 1>
    gravityOverride {z=-980}
    physicsBodyDefinitions [{
      boundBone {boneName=dyn_hoodWobble}
      boxExtents=<32 32 32>
      constraintSetup {angularLimitsMax=<15 5 5>, angularLimitsMin=<-15 -5 -5>, angularTarget=<1 0 0>}
      localJointOffset=<2 0 0>
     }]
   }
   n36 = "Constraint - dyn_backStrap_1 (PPPPPPPP)"!(
    ComponentPose=@n35
    Alpha=1
    ConstraintWeights_0=0.6601
    ConstraintWeights_1:float=0.0381
    ConstraintWeights_2:float=0.0791
    ConstraintWeights_3:float=0.043
    ConstraintWeights_4:float=0.0117
    ConstraintWeights_5:float=0.1005
    ConstraintWeights_6:float=0.0559
    ConstraintWeights_7:float=0.0117
   ) -> (Pose) {
    alphaBoolBlend {blendOption=Linear}
    alphaScaleBiasClamp {scale=0.5}
    boneToModify {boneName=dyn_backStrap_1}
    constraintSetup [
     {targetBone {boneName=spine_01}}
     {targetBone {boneName=spine_02}}
     {targetBone {boneName=spine_03}}
     {targetBone {boneName=spine_04}}
     {targetBone {boneName=dyn_tshirt_rear_l_01}}
     {targetBone {boneName=dyn_tshirt_rear_01}}
     {targetBone {boneName=dyn_tshirt_rear_02}}
     {targetBone {boneName=dyn_tshirt_rear_r_01}}
    ]
    constraintWeights [1,1,1,1,1,1,1,1]
   }
   n37 = Anim Dynamics - Bone: dyn_backStrap_1!(SimulationSpace:AnimPhysSimSpaceType Enum, ComponentPose=@n36, Alpha=1) -> (Pose) {
    angularDampingOverride=1
    angularSpringConstant=10000
    bAngularSpring=true
    bOverrideAngularDamping=true
    bUseGravityOverride=true
    gravityScale=0.35
    simulationSpace=RootRelative
    alphaBoolBlend {blendOption=Linear}
    boundBone {boneName=dyn_backStrap_1}
    componentAppliedLinearAccClamp=<2000 2000 2000>
    componentLinearAccScale=<1 1 1>
    componentLinearVelScale=<1 1 1>
    gravityOverride {z=-980}
    physicsBodyDefinitions [{
      boundBone {boneName=dyn_backStrap_1}
      boxExtents=<20 30 30>
      constraintSetup {angularLimitsMax=<0 30 0>, angularLimitsMin=<0 -30 -30>, angularTarget=<1 0 0>}
      localJointOffset=<4 0 0>
     }]
   }
   n38 = Anim Dynamics - Bone: dyn_backStrap_2!(SimulationSpace:AnimPhysSimSpaceType Enum, ComponentPose=@n37, Alpha=1) -> (Pose) {
    angularDampingOverride=1
    angularSpringConstant=10000
    bAngularSpring=true
    bOverrideAngularDamping=true
    bUseGravityOverride=true
    gravityScale=0.35
    simulationSpace=RootRelative
    alphaBoolBlend {blendOption=Linear}
    boundBone {boneName=dyn_backStrap_2}
    componentAppliedLinearAccClamp=<2000 2000 2000>
    componentLinearAccScale=<1 1 1>
    componentLinearVelScale=<1 1 1>
    gravityOverride {z=-980}
    physicsBodyDefinitions [{
      boundBone {boneName=dyn_backStrap_2}
      boxExtents=<20 30 30>
      constraintSetup {angularLimitsMax=<0 30 0>, angularLimitsMin=<0 -30 -30>, angularTarget=<1 0 0>}
      localJointOffset=<4 0 0>
     }]
   }
   Save cached pose 'PostAnimDynamics'(Pose=Component To Local(ComponentPose=@n38)) {cacheName=PostAnimDynamics}
   n39 = "Constraint - dyn_backStrap_1_MSTR (PPPPPPPP)"!(
    ComponentPose=Local To Component(LocalPose=Use cached pose 'PostDeformation'())
    Alpha=1
    ConstraintWeights_0=0.6601
    ConstraintWeights_1:float=0.0381
    ConstraintWeights_2:float=0.0791
    ConstraintWeights_3:float=0.043
    ConstraintWeights_4:float=0.0117
    ConstraintWeights_5:float=0.1005
    ConstraintWeights_6:float=0.0559
    ConstraintWeights_7:float=0.0117
   ) -> (Pose) {
    alphaBoolBlend {blendOption=Linear}
    alphaScaleBiasClamp {scale=0.5}
    boneToModify {boneName=dyn_backStrap_1_MSTR}
    constraintSetup [
     {targetBone {boneName=spine_01}}
     {targetBone {boneName=spine_02}}
     {targetBone {boneName=spine_03}}
     {targetBone {boneName=spine_04}}
     {targetBone {boneName=dyn_tshirt_rear_l_01}}
     {targetBone {boneName=dyn_tshirt_rear_01}}
     {targetBone {boneName=dyn_tshirt_rear_02}}
     {targetBone {boneName=dyn_tshirt_rear_r_01}}
    ]
    constraintWeights [1,1,1,1,1,1,1,1]
   }
   n40 = RigidBody!(ComponentPose=@n39, Alpha, OverrideWorldGravity:Vector="0,0,0", ExternalForce) -> (Pose) {
    bOverrideWorldGravity=true
    overridePhysicsAsset=&/M_MED_LemonCartCleaner_Physics
    simulationSpace=BaseBoneSpace
    alphaBoolBlend {blendOption=Linear}
    alphaScaleBiasClamp {interpSpeedDecreasing=4, interpSpeedIncreasing=4}
    baseBoneRef {boneName=spine_03}
    componentAppliedLinearAccClamp=<3000 3000 3000>
    simSpaceSettings {maxAngularAcceleration=1, maxAngularVelocity=3, maxLinearAcceleration=50, maxLinearVelocity=400, worldAlpha=1, externalLinearDragV=<0.2 0.2>}
   }
   n41 = RigidBody(ComponentPose=@n40, OverrideWorldGravity="0,0,0") {
    bOverrideWorldGravity=true
    overridePhysicsAsset=&/M_MED_LemonCartCleaner_RightSleeve_Physics
    simulationSpace=BaseBoneSpace
    alphaBoolBlend {blendOption=Linear}
    alphaScaleBiasClamp {interpSpeedDecreasing=4, interpSpeedIncreasing=4}
    baseBoneRef {boneName=upperarm_r}
    componentAppliedLinearAccClamp=<3000 3000 3000>
    simSpaceSettings {maxAngularAcceleration=1, maxAngularVelocity=5, maxLinearAcceleration=50, maxLinearVelocity=100, velocityScaleZ=0, worldAlpha=0.3}
   }
   n42 = RigidBody(ComponentPose=@n41, OverrideWorldGravity="0,0,0") {
    bOverrideWorldGravity=true
    overridePhysicsAsset=&/M_MED_LemonCartCleaner_LeftSleeve_Physics
    simulationSpace=BaseBoneSpace
    alphaBoolBlend {blendOption=Linear}
    alphaScaleBiasClamp {interpSpeedDecreasing=4, interpSpeedIncreasing=4}
    baseBoneRef {boneName=upperarm_l}
    componentAppliedLinearAccClamp=<3000 3000 3000>
    simSpaceSettings {maxAngularAcceleration=1, maxAngularVelocity=5, maxLinearAcceleration=50, maxLinearVelocity=100, velocityScaleZ=0, worldAlpha=0.3}
   }
   n43 = "Transform (Modify) Bone - Bone: dyn_tshirt_side_r_01"(ComponentPose=@n42, Alpha=1, Translation="0.001,-0.716,0.007", Rotation="0,0,0", Scale="1,1,1") {
    rotationMode=BMM_Additive
    rotationSpace=BCS_BoneSpace
    translationMode=BMM_Additive
    translationSpace=BCS_BoneSpace
    alphaBoolBlend {blendOption=Linear}
    boneToModify {boneName=dyn_tshirt_side_r_01}
    translation=<0.001 -0.716 0.007>
   }
   n44 = "Transform (Modify) Bone - Bone: dyn_tshirt_rear_r_01"(ComponentPose=@n43, Alpha=1, Translation="0,-0.5,0", Rotation="-2.316,-3.149,0.006", Scale="1,1,1") {
    rotationMode=BMM_Additive
    rotationSpace=BCS_BoneSpace
    translationMode=BMM_Additive
    translationSpace=BCS_BoneSpace
    alphaBoolBlend {blendOption=Linear}
    boneToModify {boneName=dyn_tshirt_rear_r_01}
    rotation=<rot -2.316 -3.149 0.006>
    translation {y=-0.5}
   }
   n45 = "Transform (Modify) Bone - Bone: dyn_tshirt_front_r_01"(ComponentPose=@n44, Alpha=1, Translation="0,-1,0", Rotation="-0.002,-0.327,-0.003", Scale="1,1,1") {
    rotationMode=BMM_Additive
    rotationSpace=BCS_BoneSpace
    translationMode=BMM_Additive
    translationSpace=BCS_BoneSpace
    alphaBoolBlend {blendOption=Linear}
    boneToModify {boneName=dyn_tshirt_front_r_01}
    rotation=<rot -0.002 -0.327 -0.003>
    translation {y=-1}
   }
   n46 = "Transform (Modify) Bone - Bone: dyn_R_backSleeve_1"(ComponentPose=@n45, Alpha=1, Translation="0,1.5,0", Rotation="0,0,0", Scale="1,1,1") {
    rotationMode=BMM_Additive
    rotationSpace=BCS_BoneSpace
    translationMode=BMM_Additive
    translationSpace=BCS_BoneSpace
    alphaBoolBlend {blendOption=Linear}
    boneToModify {boneName=dyn_R_backSleeve_1}
    translation {y=1.5}
   }
   n47 = "Transform (Modify) Bone - Bone: dyn_L_backSleeve_1"(ComponentPose=@n46, Alpha=1, Translation="0,-1.5,0", Rotation="0,0,0", Scale="1,1,1") {
    rotationMode=BMM_Additive
    rotationSpace=BCS_BoneSpace
    translationMode=BMM_Additive
    translationSpace=BCS_BoneSpace
    alphaBoolBlend {blendOption=Linear}
    boneToModify {boneName=dyn_L_backSleeve_1}
    translation {y=-1.5}
   }
   n48 = "Transform (Modify) Bone - Bone: dyn_backStrap_1_MSTR"(ComponentPose=@n47, Alpha=1, Translation="0,-0.5,0", Rotation="0,0,0", Scale="1,1,1") {
    rotationMode=BMM_Additive
    rotationSpace=BCS_BoneSpace
    translationMode=BMM_Additive
    translationSpace=BCS_BoneSpace
    alphaBoolBlend {blendOption=Linear}
    boneToModify {boneName=dyn_backStrap_1_MSTR}
    translation {y=-0.5}
   }
   n49 = "Transform (Modify) Bone - Bone: dyn_tshirt_side_r_02"(ComponentPose=@n48, Alpha=1, Translation="0,0,0", Rotation="0,5,0", Scale="1,1,1") {
    rotationMode=BMM_Additive
    rotationSpace=BCS_BoneSpace
    alphaBoolBlend {blendOption=Linear}
    boneToModify {boneName=dyn_tshirt_side_r_02}
    rotation {yaw=5}
   }
   n50 = "Transform (Modify) Bone - Bone: dyn_tshirt_front_r_02"(ComponentPose=@n49, Alpha=1, Translation="0,0,0", Rotation="0,10,0", Scale="1,1,1") {
    rotationMode=BMM_Additive
    rotationSpace=BCS_BoneSpace
    alphaBoolBlend {blendOption=Linear}
    boneToModify {boneName=dyn_tshirt_front_r_02}
    rotation {yaw=10}
   }
   n51 = "Transform (Modify) Bone - Bone: dyn_tshirt_front_l_02"(ComponentPose=@n50, Alpha=1, Translation="0,0,0", Rotation="0,5,0", Scale="1,1,1") {
    rotationMode=BMM_Additive
    rotationSpace=BCS_BoneSpace
    alphaBoolBlend {blendOption=Linear}
    boneToModify {boneName=dyn_tshirt_front_l_02}
    rotation {yaw=5}
   }
   n52 = "Transform (Modify) Bone - Bone: dyn_boneBelt_back_1"(ComponentPose=@n51, Alpha=1, Translation="0,0,0", Rotation="-0.010,28.256,0.000", Scale="1,1,1") {
    rotationMode=BMM_Additive
    rotationSpace=BCS_BoneSpace
    alphaBoolBlend {blendOption=Linear}
    boneToModify {boneName=dyn_boneBelt_back_1}
    rotation=<rot -0.01 28.2559 0>
   }
   Save cached pose 'PostRBAN'(Pose=Component To Local(ComponentPose=@n52)) {cacheName=PostRBAN}
  ]}
 EventGraph {nodes [
   Event BlueprintUpdateAnimation()
   > n1 = Sequence!(execute) -> (then_0, then_1, then_2:Exec, then_3:Exec, then_4:Exec, then_5:Exec)
   > n2 = Branch(Condition=OR Boolean(A=Get bIsDBNOCarried(), B=Get bIsInFrontEnd()))
   > Set simSpaceChange(simSpaceChange:AnimPhysSimSpaceType Enum=Component) -> (Output_Get:AnimPhysSimSpaceType Enum)
   Set simSpaceChange(execute=@n2.else)
   n3 = Gravity Override State Function DEPRECATED(
    execute=@n1.then_3
    CharacterPartAnimInstance=Self-Reference()
    GameTimeInSeconds=GetGameTimeInSeconds()
    GravityOverrideParams=Get Gravity Override Params()
   )
   > Set GravityOverride(GravityOverride=@n3)
   Branch(execute=@n1.then_1, Condition=Get bIsInFrontEnd())
   > Set Custom Idle is Playing Float(Custom Idle is Playing Float:float=GetCurveValue(self=GetOwnerFrontEndAnimInstance(), CurveName=Frontend_Idle_Bypass)) -> (Output_Get:float)
   n4 = Branch(execute=@n1.then_2, Condition=CheckForBackbling())
   > Set strapBackblingAlpha(strapBackblingAlpha:double=0.7) -> (Output_Get:double)
   Set strapBackblingAlpha(execute=@n4.else, strapBackblingAlpha=1)
  ]}
]
variables [
 Custom Idle is Playing Float {type=number}
 Gravity Override Params {type=GravityOverrideParamsStruct}
 GravityOverride {type=Vector}
 Is jumping Falling Alpha {type=number}
 isNotPlayingEmoteInFrontend {type=number}
 simSpaceChange {type=AnimPhysSimSpaceType}
 SleeveAlpha {type=number}
 strapBackblingAlpha {type=number}
]
